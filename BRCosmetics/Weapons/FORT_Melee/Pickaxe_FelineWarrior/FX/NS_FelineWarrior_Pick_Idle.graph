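# Niagara idle effect for the Feline Warrior pickaxe, drawing pink glow sprites along a static mesh.
# Emitters such as EpicCine and High burst one particle per triangle, pin it there each update, and cull by quality-level modulo.
# Curves indexed by normalized particle index shape color, alpha and size along the mesh.
# Camera-angle dot products stretch the aligned sprites and, in EpicCine, fade them, with separate front-end and gameplay tuning.
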NiagaraSystem /BRCosmetics/Weapons/FORT_Melee/Pickaxe_FelineWarrior/FX/NS_FelineWarrior_Pick_Idle
dependencies {
 usedDataInterfaces [
  &NiagaraDataInterfaceCurve
  &NiagaraDataInterfaceColorCurve
  &NiagaraDataInterfacePlatformSet
  &NiagaraDataInterfaceCamera
  &NiagaraDataInterfaceStaticMesh
 ]
 usedDynamicInputs []
 usedModules [
  &self:SystemSpawnScript.SystemScriptSource.SystemScriptGraph.NiagaraNodeAssignment_0.SetVariables_#1
  &self:SystemSpawnScript.SystemScriptSource.SystemScriptGraph.NiagaraNodeAssignment_1.SetVariables_#2
  &SystemState
  &Timeline
 ]
}
emitters {
 EpicCine {
  emitterSpawnScript [
   /SetVariables(Emitter.Num Particles To Spawn:NiagaraInt32=&self:Get_num_Tris(Static Mesh:NiagaraDataInterfaceStaticMesh=$System.Static Mesh)) {isSetParameters=true, module=SetVariables_#3}
   /SetVariables(Emitter.Plat epicCin=5) {isSetParameters=true, module=SetVariables_#4}
   /SetVariables(Emitter.Curve for Floats=DataInterface() {propertyValues {Curve {keys [
        {arriveTangent=3.4855, interpMode=RCIM_Cubic, leaveTangent=3.4855, leaveTangentWeight=0.1333, tangentMode=RCTM_User, time=0, value=0.803}
        {arriveTangent=6.595, interpMode=RCIM_Cubic, leaveTangent=6.595, time=0.0535, value=1.0629}
        {
         arriveTangent=0.6084
         arriveTangentWeight=0.0465
         interpMode=RCIM_Cubic
         leaveTangentWeight=0.0869
         tangentMode=RCTM_Break
         time=0.1234
         value=1.6172
        }
        {arriveTangent=-2.0983, interpMode=RCIM_Cubic, leaveTangent=-2.0983, time=0.3219, value=1.6496}
        {arriveTangentWeight=0.0059, interpMode=RCIM_Cubic, leaveTangent=5.4842, tangentMode=RCTM_Break, time=0.4176, value=1}
        {arriveTangentWeight=0.0581, interpMode=RCIM_Cubic, tangentMode=RCTM_User, time=0.5918, value=1.6527}
        {arriveTangent=-4.4995, arriveTangentWeight=0.1361, interpMode=RCIM_Cubic, leaveTangent=-4.4995, tangentMode=RCTM_User, time=1, value=1}
       ]}}}) {isSetParameters=true}
  ]
  emitterUpdateScript [
   &EmitterState()
   &SpawnBurst_Instantaneous(Spawn Count=$Emitter.Num Particles To Spawn, Use Spawn Probability=false, UseLoopCountLimit=true)
  ]
  particleSpawnScript [
   &InitializeParticle(
    Color Mode=::NewEnumerator0
    Lifetime=5
    Sprite Size Mode=::NewEnumerator3
    Uniform Sprite Size=8
    UsePositionOffset=false
    Write Material Random=true
   )
   &PartitionParticles(
    Normalized Index In Partition Includes One=true
    Number Of Partitions=1
    Particles Per Partition:NiagaraInt32=$Emitter.Num Particles To Spawn
    Partition Mode:ENiagara_PartitionMode=::NewEnumerator1
   )
   &Color(Color=&MultiplyLinearColor(
     Linear Color=&ColorFromCurve(
      ColorCurve=DataInterface() {propertyValues {
        AlphaCurve {keys [
          <at 0 0.2644>
          <at 0.0941 0.4935>
          <at 0.1935 0.2316>
          <at 0.2965 0.2316>
          {arriveTangentWeight=0.0471, interpMode=RCIM_Cubic, leaveTangentWeight=0.0162, tangentMode=RCTM_User, time=0.4002, value=0.1529}
          {arriveTangentWeight=0.1704, interpMode=RCIM_Cubic, leaveTangentWeight=0.1641, tangentMode=RCTM_User, time=0.4426, value=0.1529}
          <at 0.503 1>
          {arriveTangent=-1.9219, interpMode=RCIM_Cubic, leaveTangent=-1.9219, time=0.6971, value=1}
          <at 0.9438 0.1529 RCIM_Cubic>
         ]}
        BlueCurve {keys [
          {interpMode=RCIM_Cubic, leaveTangentWeight=0.066, tangentMode=RCTM_User, time=0.3582, value=0.9242}
          {
           arriveTangent=-0.0196
           arriveTangentWeight=0.0526
           interpMode=RCIM_Cubic
           leaveTangent=-0.0196
           leaveTangentWeight=0.0879
           time=0.5128
           value=0.9162
          }
          {
           arriveTangent=0.0166
           arriveTangentWeight=0.0874
           interpMode=RCIM_Cubic
           leaveTangent=0.0166
           leaveTangentWeight=0.0774
           time=0.7712
           value=0.9162
          }
          {arriveTangentWeight=0.0763, interpMode=RCIM_Cubic, tangentMode=RCTM_User, time=1, value=0.9242}
         ]}
        GreenCurve {keys [
          {arriveTangent=-4.0724, interpMode=RCIM_Cubic, leaveTangent=-4.0724, leaveTangentWeight=0.066, tangentMode=RCTM_User, time=0.3582, value=0.2}
          {arriveTangentWeight=0.1478, interpMode=RCIM_Cubic, leaveTangentWeight=0.1152, tangentMode=RCTM_User, time=0.5128, value=0.1146}
          {arriveTangentWeight=0.1079, interpMode=RCIM_Cubic, leaveTangentWeight=0.1599, tangentMode=RCTM_User, time=0.7712, value=0.1146}
          {arriveTangent=4.0724, arriveTangentWeight=0.0763, interpMode=RCIM_Cubic, leaveTangent=4.0724, tangentMode=RCTM_User, time=1, value=0.2}
         ]}
        RedCurve {keys [
          {interpMode=RCIM_Cubic, leaveTangentWeight=0.066, tangentMode=RCTM_User, time=0.3582, value=1}
          {arriveTangentWeight=0.066, interpMode=RCIM_Cubic, leaveTangentWeight=0.0515, tangentMode=RCTM_User, time=0.5128, value=1}
          {arriveTangentWeight=0.0515, interpMode=RCIM_Cubic, leaveTangentWeight=0.0763, tangentMode=RCTM_User, time=0.7712, value=1}
          {arriveTangentWeight=0.0763, interpMode=RCIM_Cubic, tangentMode=RCTM_User, time=1, value=1}
         ]}
       }}
      CurveIndex=$Particles.PartitionParticles.Normalized Index In Partition
     )
     Scale Factor=&MakeFloat4FromVectorAndFloat("Vector (RGB)"=&VectorFromFloat(Value=2), "Float (Alpha)"=0.4)
    ))
   &CameraOffset(Camera Offset Amount=7)
   &KillParticles(Kill Particles=&SetBoolByIntComparison(
     A=&ModuloInt(
      A=&ReturnParticleIDAttributes(Particle ID Attribute=::NewEnumerator1)
      B=&MakeCustomIntFromBool(
       Bool=&IsPlatformSetActive(New Platform Set=DataInterface() {propertyValues {Platforms {qualityLevelMask=-31}}})
       True INT=$Emitter.Plat Low
       False INT=&MakeCustomIntFromBool(
        Bool=&IsPlatformSetActive(New Platform Set=DataInterface() {propertyValues {Platforms {qualityLevelMask=-30}}})
        True INT=$Emitter.PLat med
        False INT=&MakeCustomIntFromBool(
         Bool=&IsPlatformSetActive(New Platform Set=DataInterface() {propertyValues {Platforms {qualityLevelMask=-28}}})
         True INT=$Emitter.PLat High
         False INT=&MakeCustomIntFromBool(
          Bool=&IsPlatformSetActive(New Platform Set=DataInterface() {propertyValues {Platforms {qualityLevelMask=-8}}})
          True INT=$Emitter.Plat epicCin
          False INT=2147483647
         )
        )
       )
      )
     )
     B=0
     Comparison Type=::NewEnumerator5
    ))
   &KillParticles(Kill Particles=&SetBoolByFloatComparison(
     A=&MakeFloatFromLinearColor(LinearColor=$Particles.Color, Channel:ENiagaraLinearColor_Channels=::NewEnumerator3)
     B=0
     Comparison Type=::NewEnumerator7
    )) {module=KillParticles002}
   &KillParticles(Kill Particles=&SetBoolByIntComparison(A=&ReturnParticleIDAttributes(), B=10, Comparison Type=::NewEnumerator4)) {module=KillParticles003}
   /SetVariables(Particles.PLacement:NiagaraInt32=&ReturnParticleIDAttributes()) {isSetParameters=true}
  ]
  particleUpdateScript [
   &ParticleState(Kill Particles When Lifetime Has Elapsed=false, Let Infinitely Lived Particles Die When Emitter Deactivates=true)
   &ScaleSpriteSize(
    Scale Sprite Size Mode=::NewEnumerator0
    Uniform Scale Factor=&FloatFromCurve(FloatCurve=$Emitter.Curve for Floats, CurveIndex=$Particles.PartitionParticles.Normalized Index In Partition)
   ) {module=ScaleSpriteSize001}
   &ScaleSpriteSize(
    Scale Factor=&FN_FrontEndVector2Modifier(
     Front End:Vector2f=&MakeVector2D(
      X=&Lerp_Float(
       A=&FloatFromCurve(
        FloatCurve=DataInterface() {propertyValues {Curve {keys [
            <at 0 1 RCIM_Cubic>
            {arriveTangentWeight=0.0507, interpMode=RCIM_Cubic, leaveTangentWeight=0.1376, tangentMode=RCTM_User, time=0.1193, value=2}
            {arriveTangentWeight=0.1467, interpMode=RCIM_Cubic, leaveTangentWeight=0.0257, tangentMode=RCTM_User, time=0.4433, value=2}
            {arriveTangentWeight=0.0257, interpMode=RCIM_Cubic, leaveTangentWeight=0.122, tangentMode=RCTM_User, time=0.5, value=1}
            <at 0.7693 1.05 RCIM_Cubic>
            <at 1 1 RCIM_Cubic>
           ]}}}
        CurveIndex=$Particles.PartitionParticles.Normalized Index In Partition
       )
       B=&FloatFromCurve(
        FloatCurve=DataInterface() {propertyValues {Curve {keys [
            <at 0 1 RCIM_Cubic>
            {arriveTangentWeight=0.0795, interpMode=RCIM_Cubic, leaveTangentWeight=0.1591, tangentMode=RCTM_User, time=0.1, value=2}
            {arriveTangentWeight=0.1272, interpMode=RCIM_Cubic, leaveTangentWeight=0.1272, tangentMode=RCTM_User, time=0.3, value=2}
            {arriveTangent=-1.4286, interpMode=RCIM_Cubic, leaveTangent=-1.4286, time=0.5, value=1}
            <at 1 1>
           ]}}}
        CurveIndex=$Particles.PartitionParticles.Normalized Index In Partition
       )
       Alpha=&FloatFromCurve(
        FloatCurve=DataInterface() {propertyValues {Curve {keys [<at 0.15 1>, {arriveTangent=1.8365, leaveTangent=1.8365, time=0.5, value=0}]}}}
        CurveIndex=&Abs_Float(Float:NiagaraFloat=&DotProduct(Vector A=&ReturnCameraVector(Camera Vectors=::NewEnumerator1, Coordinate Space=::World), Vector B=<0 0 1>))
       )
      )
      Y=&Lerp_Float(
       A=1
       B=&FN_FrontEndFloatModifier(
        Gameplay=&FloatFromCurve(
         FloatCurve=DataInterface() {propertyValues {Curve {keys [
             {
              interpMode=RCIM_Cubic
              leaveTangent=9.7835
              leaveTangentWeight=0.6125
              tangentMode=RCTM_Break
              tangentWeightMode=RCTWM_WeightedBoth
              time=0
              value=3.5431
             }
             {arriveTangent=-18.6565, interpMode=RCIM_Cubic, leaveTangent=-18.6565, time=0.1275, value=0.3703}
             {
              arriveTangent=42.0983
              arriveTangentWeight=0.0581
              interpMode=RCIM_Cubic
              leaveTangent=42.0983
              leaveTangentWeight=0.0669
              time=0.1303
              value=1.1118
             }
             {arriveTangent=4.0429, interpMode=RCIM_Cubic, leaveTangent=4.0429, time=0.2613, value=6}
             {arriveTangentWeight=0.0651, interpMode=RCIM_Cubic, leaveTangentWeight=0.0966, tangentMode=RCTM_User, time=0.35, value=2}
             {arriveTangentWeight=0.0581, interpMode=RCIM_Cubic, leaveTangentWeight=0.0623, tangentMode=RCTM_User, time=0.6, value=2.8}
             {arriveTangentWeight=0.0737, interpMode=RCIM_Cubic, tangentMode=RCTM_User, time=1, value=3}
            ]}}}
         CurveIndex=$Particles.PartitionParticles.Normalized Index In Partition
        )
        Front End=&FloatFromCurve(
         FloatCurve=DataInterface() {propertyValues {Curve {keys [
             {
              interpMode=RCIM_Cubic
              leaveTangent=9.7835
              leaveTangentWeight=0.6125
              tangentMode=RCTM_Break
              tangentWeightMode=RCTWM_WeightedBoth
              time=0
              value=3.5431
             }
             {arriveTangent=-0.5713, interpMode=RCIM_Cubic, leaveTangent=-0.5713, time=0.1275, value=0.3703}
             {
              arriveTangent=42.0982
              arriveTangentWeight=0.0581
              interpMode=RCIM_Cubic
              leaveTangent=42.0982
              leaveTangentWeight=0.0669
              time=0.1303
              value=3.4686
             }
             {arriveTangent=-6.6853, interpMode=RCIM_Cubic, leaveTangent=-6.6853, time=0.2613, value=6}
             {arriveTangentWeight=0.0651, interpMode=RCIM_Cubic, leaveTangentWeight=0.0966, tangentMode=RCTM_User, time=0.35, value=2}
             {arriveTangentWeight=0.0581, interpMode=RCIM_Cubic, leaveTangentWeight=0.0623, tangentMode=RCTM_User, time=0.6, value=2.8}
             {arriveTangentWeight=0.0737, interpMode=RCIM_Cubic, tangentMode=RCTM_User, time=1, value=3}
            ]}}}
         CurveIndex=$Particles.PartitionParticles.Normalized Index In Partition
        )
       )
       Alpha=&FloatFromCurve(
        FloatCurve=DataInterface() {propertyValues {Curve {keys [
            {
             arriveTangentWeight=0.616
             interpMode=RCIM_Cubic
             leaveTangentWeight=0.616
             tangentMode=RCTM_User
             tangentWeightMode=RCTWM_WeightedBoth
             time=0
             value=1
            }
            {
             arriveTangent=-3.3955
             arriveTangentWeight=0.688
             interpMode=RCIM_Cubic
             leaveTangent=-3.3955
             leaveTangentWeight=0.688
             tangentMode=RCTM_User
             tangentWeightMode=RCTWM_WeightedBoth
             time=1
             value=0
            }
           ]}}}
        CurveIndex=&Abs_Float(Float:NiagaraFloat=&DotProduct(
          Vector A=&ReturnMeshOrientationAxis(Mesh Orientation Axis=::ZAxis, Mesh Orientation Quat=$Particles.MeshOrientation)
          Vector B=&ReturnCameraVector(Camera Vectors=::NewEnumerator4, Coordinate Space=::Simulation, Origin Position=$Particles.Position)
         ))
       )
      )
     )
     Gameplay:Vector2f=&MakeVector2D(
      X=&Lerp_Float(
       A=&FloatFromCurve(
        FloatCurve=DataInterface() {propertyValues {Curve {keys [
            <at 0 1 RCIM_Cubic>
            {arriveTangentWeight=0.0507, interpMode=RCIM_Cubic, leaveTangentWeight=0.1376, tangentMode=RCTM_User, time=0.1193, value=2}
            {arriveTangentWeight=0.1467, interpMode=RCIM_Cubic, leaveTangentWeight=0.0257, tangentMode=RCTM_User, time=0.4433, value=2}
            {arriveTangentWeight=0.0257, interpMode=RCIM_Cubic, leaveTangentWeight=0.122, tangentMode=RCTM_User, time=0.5, value=1}
            <at 0.7693 1.05 RCIM_Cubic>
            <at 1 1 RCIM_Cubic>
           ]}}}
        CurveIndex=$Particles.PartitionParticles.Normalized Index In Partition
       )
       B=&FloatFromCurve(
        FloatCurve=DataInterface() {propertyValues {Curve {keys [
            <at 0 1 RCIM_Cubic>
            {arriveTangentWeight=0.0795, interpMode=RCIM_Cubic, leaveTangentWeight=0.1591, tangentMode=RCTM_User, time=0.1, value=0.5}
            {arriveTangentWeight=0.1272, interpMode=RCIM_Cubic, leaveTangentWeight=0.1272, tangentMode=RCTM_User, time=0.3, value=0.5}
            {arriveTangent=0.7143, interpMode=RCIM_Cubic, leaveTangent=0.7143, time=0.5, value=1}
            <at 1 1>
           ]}}}
        CurveIndex=$Particles.PartitionParticles.Normalized Index In Partition
       )
       Alpha=&FloatFromCurve(
        FloatCurve=DataInterface() {propertyValues {Curve {keys [<at 0.15 1>, {arriveTangent=1.8365, leaveTangent=1.8365, time=0.5, value=0}]}}}
        CurveIndex=&Abs_Float(Float:NiagaraFloat=&DotProduct(Vector A=&ReturnCameraVector(Camera Vectors=::NewEnumerator1, Coordinate Space=::World), Vector B=<0 0 1>))
       )
      )
      Y=&Lerp_Float(
       A=1
       B=&FloatFromCurve(
        FloatCurve=DataInterface() {propertyValues {Curve {keys [
            {
             interpMode=RCIM_Cubic
             leaveTangent=9.7835
             leaveTangentWeight=0.6125
             tangentMode=RCTM_Break
             tangentWeightMode=RCTWM_WeightedBoth
             time=0
             value=3.5431
            }
            {arriveTangent=-18.6565, interpMode=RCIM_Cubic, leaveTangent=-18.6565, time=0.1275, value=0.3703}
            {
             arriveTangent=42.0983
             arriveTangentWeight=0.0581
             interpMode=RCIM_Cubic
             leaveTangent=42.0983
             leaveTangentWeight=0.0669
             time=0.1303
             value=1.1118
            }
            {arriveTangent=4.0429, interpMode=RCIM_Cubic, leaveTangent=4.0429, time=0.2613, value=6}
            {arriveTangentWeight=0.0651, interpMode=RCIM_Cubic, leaveTangentWeight=0.0966, tangentMode=RCTM_User, time=0.35, value=2}
            {arriveTangentWeight=0.0581, interpMode=RCIM_Cubic, leaveTangentWeight=0.0623, tangentMode=RCTM_User, time=0.6, value=2.8}
            {arriveTangentWeight=0.0737, interpMode=RCIM_Cubic, tangentMode=RCTM_User, time=1, value=3}
           ]}}}
        CurveIndex=$Particles.PartitionParticles.Normalized Index In Partition
       )
       Alpha=&FloatFromCurve(
        FloatCurve=DataInterface() {propertyValues {Curve {keys [
            {
             arriveTangentWeight=0.616
             interpMode=RCIM_Cubic
             leaveTangentWeight=0.616
             tangentMode=RCTM_User
             tangentWeightMode=RCTWM_WeightedBoth
             time=0
             value=1
            }
            {
             arriveTangent=-3.3955
             arriveTangentWeight=0.688
             interpMode=RCIM_Cubic
             leaveTangent=-3.3955
             leaveTangentWeight=0.688
             tangentMode=RCTM_User
             tangentWeightMode=RCTWM_WeightedBoth
             time=1
             value=0
            }
           ]}}}
        CurveIndex=&Abs_Float(Float:NiagaraFloat=&DotProduct(
          Vector A=&ReturnMeshOrientationAxis(Mesh Orientation Axis=::ZAxis, Mesh Orientation Quat=$Particles.MeshOrientation)
          Vector B=&ReturnCameraVector(Camera Vectors=::NewEnumerator4, Coordinate Space=::Simulation, Origin Position=$Particles.Position)
         ))
       )
      )
     )
    )
    Scale Sprite Size Mode=::NewEnumerator2
   )
   /Niagara/Modules/Spawn/Location/StaticMeshLocation(
    FixedBarycentric=false
    MaskSection=false
    MaskSpawnGroup=false
    OffsetAlongNormal=false
    SamplePosOffset=false
    Static Mesh:NiagaraDataInterfaceStaticMesh=$System.Static Mesh
    Triangle ID:NiagaraInt32=$Particles.PartitionParticles.Index In Partition
    Triangle Sampling Mode:ENiagara_StaticTriangleMode=::NewEnumerator2
   )
   &SolveForcesAndVelocity(Clamp Velocity=false, Limit Acceleration=false)
   &SpriteFacingAndAlignment(
    Sprite Alignment=&ReturnMeshOrientationAxis(Mesh Orientation Axis=::ZAxis, Mesh Orientation Quat=$Particles.MeshOrientation)
    Sprite Facing=&ReturnCameraVector(Camera Vectors=::NewEnumerator4, Coordinate Space=::Simulation, Origin Position=$Particles.Position)
    WriteAlignment=true
    WriteFacing=true
   )
   /SetVariables(Particles.Velocity=<1 0 0>) {isSetParameters=true, module=SetVariables_#5}
   &FN_Light_Attributes(
    Create Color Attribute=true
    Create Exponent Attribute=true
    Create Position Attribute=true
    Create Radius Attribute=true
    Enable Particle Light Enable=true
   )
   &ScaleColor(Scale RGB=&VectorFromFloat(Value=3), ScaleA=true, ScaleRGB=true)
   &KillParticles(Kill Particles=&SetBoolByIntComparison(A=$Particles.PartitionParticles.Index In Partition, B=10, Comparison Type=::NewEnumerator4)) {module=KillParticles004}
   &ScaleColor(
    Scale Alpha=&FloatFromCurve(
     FloatCurve=DataInterface() {propertyValues {Curve {keys [
         {
          arriveTangent=-0.3625
          arriveTangentWeight=0.6683
          interpMode=RCIM_Cubic
          leaveTangent=-0.3625
          leaveTangentWeight=0.6683
          tangentMode=RCTM_User
          tangentWeightMode=RCTWM_WeightedBoth
          time=0
          value=1
         }
         {
          arriveTangent=-1.2454
          arriveTangentWeight=0.2827
          interpMode=RCIM_Cubic
          leaveTangent=-1.2454
          leaveTangentWeight=0.2827
          tangentMode=RCTM_User
          tangentWeightMode=RCTWM_WeightedBoth
          time=1
          value=0.1
         }
        ]}}}
     CurveIndex=&Abs_Float(Float:NiagaraFloat=&DotProduct(
       Vector A=&ReturnMeshOrientationAxis(Mesh Orientation Axis=::ZAxis, Mesh Orientation Quat=$Particles.MeshOrientation)
       Vector B=&ReturnCameraVector(Camera Vectors=::NewEnumerator4, Coordinate Space=::Simulation, Origin Position=$Particles.Position)
      ))
    )
    ScaleA=true
    ScaleRGB=true
   ) {module=ScaleColor001}
   &Color(Color=&ColorFromCurve(
     ColorCurve=DataInterface() {propertyValues {
       AlphaCurve {keys [<at 1 1 RCIM_Cubic>]}
       BlueCurve {keys [
         <at 0 1 RCIM_Cubic>
         {arriveTangent=-6.6667, interpMode=RCIM_Cubic, leaveTangent=-6.6667, time=0, value=0.5}
         {arriveTangent=-1, interpMode=RCIM_Cubic, leaveTangent=-1, time=0.15, value=0}
         {interpMode=RCIM_Cubic, leaveTangent=1.4286, time=0.5, value=0}
         <at 0.9875 1 RCIM_Cubic>
        ]}
       GreenCurve {keys [
         <at 0 0 RCIM_Cubic>
         <at 0 0 RCIM_Cubic>
         {arriveTangent=2, interpMode=RCIM_Cubic, leaveTangent=2, time=0.15, value=0}
         <at 0.5 1 RCIM_Cubic>
         {interpMode=RCIM_Cubic, leaveTangent=-1.5385, time=0.9875, value=0}
        ]}
       RedCurve {keys [
         <at 0 0 RCIM_Cubic>
         {arriveTangent=6.6667, interpMode=RCIM_Cubic, leaveTangent=6.6667, time=0, value=1}
         {arriveTangent=-2, interpMode=RCIM_Cubic, leaveTangent=-2, time=0.15, value=1}
         {interpMode=RCIM_Cubic, leaveTangent=-1.4286, time=0.5, value=0}
         {interpMode=RCIM_Cubic, leaveTangent=1.5385, time=0.9875, value=0}
        ]}
      }}
     CurveIndex=&FloatFromCurve(FloatCurve=$Emitter.Curve for Floats, CurveIndex=$Particles.PartitionParticles.Normalized Index In Partition)
    )) {enabled=false, module=Color001}
   &ScaleColor(
    Scale Alpha=&FN_FrontEndFloatModifier(
     Gameplay=&FloatFromCurve(
      FloatCurve=DataInterface() {propertyValues {Curve {keys [
          {arriveTangentWeight=0.0535, interpMode=RCIM_Cubic, leaveTangentWeight=0.107, tangentMode=RCTM_User, time=0.1899, value=1.1}
          {arriveTangentWeight=0.102, interpMode=RCIM_Cubic, leaveTangentWeight=0.2368, tangentMode=RCTM_User, time=0.1968, value=0.65}
          {arriveTangentWeight=0.0916, interpMode=RCIM_Cubic, leaveTangentWeight=0.0916, tangentMode=RCTM_User, time=0.3713, value=0.65}
          <at 0.4132 1.1>
         ]}}}
      CurveIndex=$Particles.PartitionParticles.Normalized Index In Partition
     )
     Front End=&FloatFromCurve(
      FloatCurve=DataInterface() {propertyValues {Curve {keys [
          {arriveTangentWeight=0.0535, interpMode=RCIM_Cubic, leaveTangentWeight=0.107, tangentMode=RCTM_User, time=0.1899, value=1.1}
          {arriveTangentWeight=0.102, interpMode=RCIM_Cubic, leaveTangentWeight=0.2368, tangentMode=RCTM_User, time=0.1968, value=1}
          {arriveTangentWeight=0.0916, interpMode=RCIM_Cubic, leaveTangentWeight=0.0916, tangentMode=RCTM_User, time=0.3713, value=1}
          <at 0.4132 1.1>
         ]}}}
      CurveIndex=$Particles.PartitionParticles.Normalized Index In Partition
     )
    )
    ScaleA=true
    ScaleRGB=true
   ) {module=ScaleColor002}
  ]
  rendererClasses [&NiagaraSpriteRendererProperties]
  renderers [
   {rendererClass=&NiagaraSpriteRendererProperties}
   {rendererClass=&NiagaraSpriteRendererProperties, rendererIndex=1}
  ]
 }
 High {
  emitterSpawnScript [
   /SetVariables(Emitter.Num Particles To Spawn:NiagaraInt32=&self:Get_num_Tris(Static Mesh:NiagaraDataInterfaceStaticMesh=$System.Static Mesh)) {isSetParameters=true, module=SetVariables_#3}
   /SetVariables(Emitter.PLat High=6) {isSetParameters=true, module=SetVariables_#4}
  ]
  emitterUpdateScript [
   &EmitterState()
   &SpawnBurst_Instantaneous(Spawn Count=$Emitter.Num Particles To Spawn, Use Spawn Probability=false, UseLoopCountLimit=true)
  ]
  particleSpawnScript [
   &InitializeParticle(
    Color Mode=::NewEnumerator0
    Lifetime=5
    Sprite Size Mode=::NewEnumerator3
    Uniform Sprite Size=8
    UsePositionOffset=false
    Write Material Random=true
   )
   &PartitionParticles(
    Normalized Index In Partition Includes One=true
    Number Of Partitions=1
    Particles Per Partition:NiagaraInt32=$Emitter.Num Particles To Spawn
    Partition Mode:ENiagara_PartitionMode=::NewEnumerator1
   )
   &Color(Color=&MultiplyLinearColor(
     Linear Color=&ColorFromCurve(
      ColorCurve=DataInterface() {propertyValues {
        AlphaCurve {keys [
          <at 0 0.1317>
          <at 0.0941 0.4021>
          <at 0.1935 0.093>
          <at 0.2965 0.093>
          {arriveTangentWeight=0.0471, interpMode=RCIM_Cubic, leaveTangentWeight=0.0162, tangentMode=RCTM_User, time=0.4002, value=0}
          {arriveTangentWeight=0.1704, interpMode=RCIM_Cubic, leaveTangentWeight=0.1641, tangentMode=RCTM_User, time=0.4426, value=0}
          <at 0.503 0.4663>
          {arriveTangent=-1.0579, interpMode=RCIM_Cubic, leaveTangent=-1.0579, time=0.6971, value=0.4663}
          <at 0.9438 0 RCIM_Cubic>
         ]}
        BlueCurve {keys [
          {interpMode=RCIM_Cubic, leaveTangentWeight=0.066, tangentMode=RCTM_User, time=0.3582, value=0.9242}
          {
           arriveTangent=-0.0196
           arriveTangentWeight=0.0526
           interpMode=RCIM_Cubic
           leaveTangent=-0.0196
           leaveTangentWeight=0.0879
           time=0.5128
           value=0.9162
          }
          {
           arriveTangent=0.0166
           arriveTangentWeight=0.0874
           interpMode=RCIM_Cubic
           leaveTangent=0.0166
           leaveTangentWeight=0.0774
           time=0.7712
           value=0.9162
          }
          {arriveTangentWeight=0.0763, interpMode=RCIM_Cubic, tangentMode=RCTM_User, time=1, value=0.9242}
         ]}
        GreenCurve {keys [
          {arriveTangent=-4.0724, interpMode=RCIM_Cubic, leaveTangent=-4.0724, leaveTangentWeight=0.066, tangentMode=RCTM_User, time=0.3582, value=0.2}
          {arriveTangentWeight=0.1478, interpMode=RCIM_Cubic, leaveTangentWeight=0.1152, tangentMode=RCTM_User, time=0.5128, value=0.1146}
          {arriveTangentWeight=0.1079, interpMode=RCIM_Cubic, leaveTangentWeight=0.1599, tangentMode=RCTM_User, time=0.7712, value=0.1146}
          {arriveTangent=4.0724, arriveTangentWeight=0.0763, interpMode=RCIM_Cubic, leaveTangent=4.0724, tangentMode=RCTM_User, time=1, value=0.2}
         ]}
        RedCurve {keys [
          {interpMode=RCIM_Cubic, leaveTangentWeight=0.066, tangentMode=RCTM_User, time=0.3582, value=1}
          {arriveTangentWeight=0.066, interpMode=RCIM_Cubic, leaveTangentWeight=0.0515, tangentMode=RCTM_User, time=0.5128, value=1}
          {arriveTangentWeight=0.0515, interpMode=RCIM_Cubic, leaveTangentWeight=0.0763, tangentMode=RCTM_User, time=0.7712, value=1}
          {arriveTangentWeight=0.0763, interpMode=RCIM_Cubic, tangentMode=RCTM_User, time=1, value=1}
         ]}
       }}
      CurveIndex=$Particles.PartitionParticles.Normalized Index In Partition
     )
     Scale Factor=&MakeFloat4FromVectorAndFloat("Vector (RGB)"=&VectorFromFloat(Value=3), "Float (Alpha)"=0.4)
    ))
   &KillParticles(Kill Particles=&SetBoolByIntComparison(
     A=&ModuloInt(
      A=&ReturnParticleIDAttributes(Particle ID Attribute=::NewEnumerator1)
      B=&MakeCustomIntFromBool(
       Bool=&IsPlatformSetActive(New Platform Set=DataInterface() {propertyValues {Platforms {qualityLevelMask=-31}}})
       True INT=$Emitter.Plat Low
       False INT=&MakeCustomIntFromBool(
        Bool=&IsPlatformSetActive(New Platform Set=DataInterface() {propertyValues {Platforms {qualityLevelMask=-30}}})
        True INT=$Emitter.PLat med
        False INT=&MakeCustomIntFromBool(
         Bool=&IsPlatformSetActive(New Platform Set=DataInterface() {propertyValues {Platforms {qualityLevelMask=-28}}})
         True INT=$Emitter.PLat High
         False INT=&MakeCustomIntFromBool(
          Bool=&IsPlatformSetActive(New Platform Set=DataInterface() {propertyValues {Platforms {qualityLevelMask=-8}}})
          True INT=$Emitter.Plat epicCin
          False INT=2147483647
         )
        )
       )
      )
     )
     B=0
     Comparison Type=::NewEnumerator5
    ))
   &KillParticles(Kill Particles=&Boolean_AndOperation(
     Bool1=&IsPlatformSetActive(New Platform Set=DataInterface() {propertyValues {Platforms {qualityLevelMask=-31}}})
     Bool2=&SetBoolByIntComparison(
      A=&ReturnParticleIDAttributes(Particle ID Attribute=::NewEnumerator1)
      B=0
      Comparison Type=::NewEnumerator4
     )
    )) {module=KillParticles001}
   &CameraOffset(Camera Offset Amount=7)
   &KillParticles(Kill Particles=&SetBoolByFloatComparison(
     A=&MakeFloatFromLinearColor(LinearColor=$Particles.Color, Channel:ENiagaraLinearColor_Channels=::NewEnumerator3)
     B=0
     Comparison Type=::NewEnumerator7
    )) {module=KillParticles002}
  ]
  particleUpdateScript [
   &ParticleState(Kill Particles When Lifetime Has Elapsed=false, Let Infinitely Lived Particles Die When Emitter Deactivates=true)
   &ScaleSpriteSize(
    Scale Sprite Size Mode=::NewEnumerator0
    Uniform Scale Factor=&FloatFromCurve(
     FloatCurve=DataInterface() {propertyValues {Curve {keys [
         {arriveTangent=3.4855, interpMode=RCIM_Cubic, leaveTangent=3.4855, leaveTangentWeight=0.1686, tangentMode=RCTM_User, time=0, value=0.65}
         {
          arriveTangent=0.6084
          arriveTangentWeight=0.0544
          interpMode=RCIM_Cubic
          leaveTangentWeight=0.0927
          tangentMode=RCTM_Break
          time=0.1394
          value=1
         }
         {
          arriveTangentWeight=0.0927
          interpMode=RCIM_Cubic
          leaveTangent=5.4842
          leaveTangentWeight=0.3238
          tangentMode=RCTM_Break
          time=0.4176
          value=1
         }
         {arriveTangentWeight=0.0581, interpMode=RCIM_Cubic, leaveTangentWeight=0.0623, tangentMode=RCTM_User, time=0.5918, value=1.6527}
         {arriveTangent=-1.5358, arriveTangentWeight=0.0737, interpMode=RCIM_Cubic, leaveTangent=-1.5358, tangentMode=RCTM_User, time=1, value=1}
        ]}}}
     CurveIndex=$Particles.PartitionParticles.Normalized Index In Partition
    )
   ) {module=ScaleSpriteSize001}
   &ScaleSpriteSize(
    Scale Factor=&FN_FrontEndVector2Modifier(
     Front End:Vector2f=&MakeVector2D(
      X=&Lerp_Float(
       A=&FloatFromCurve(
        FloatCurve=DataInterface() {propertyValues {Curve {keys [
            <at 0 1 RCIM_Cubic>
            {arriveTangentWeight=0.0507, interpMode=RCIM_Cubic, leaveTangentWeight=0.1376, tangentMode=RCTM_User, time=0.1193, value=2}
            {arriveTangentWeight=0.1467, interpMode=RCIM_Cubic, leaveTangentWeight=0.0257, tangentMode=RCTM_User, time=0.4433, value=2}
            {arriveTangentWeight=0.0257, interpMode=RCIM_Cubic, leaveTangentWeight=0.122, tangentMode=RCTM_User, time=0.5, value=1}
            <at 0.7693 1.05 RCIM_Cubic>
            <at 1 1 RCIM_Cubic>
           ]}}}
        CurveIndex=$Particles.PartitionParticles.Normalized Index In Partition
       )
       B=&FloatFromCurve(
        FloatCurve=DataInterface() {propertyValues {Curve {keys [
            <at 0 1 RCIM_Cubic>
            {arriveTangentWeight=0.0795, interpMode=RCIM_Cubic, leaveTangentWeight=0.1591, tangentMode=RCTM_User, time=0.0438, value=2.4221}
            {arriveTangentWeight=0.1272, interpMode=RCIM_Cubic, leaveTangentWeight=0.1272, tangentMode=RCTM_User, time=0.2438, value=2.4221}
            {arriveTangent=-1.8806, interpMode=RCIM_Cubic, leaveTangent=-1.8806, time=0.411, value=1.1371}
            <at 1 1>
           ]}}}
        CurveIndex=$Particles.PartitionParticles.Normalized Index In Partition
        Scale Curve=1.25
       )
       Alpha=&FloatFromCurve(
        FloatCurve=DataInterface() {propertyValues {Curve {keys [<at 0.15 1>, {arriveTangent=1.8365, leaveTangent=1.8365, time=0.5, value=0}]}}}
        CurveIndex=&Abs_Float(Float:NiagaraFloat=&DotProduct(Vector A=&ReturnCameraVector(Camera Vectors=::NewEnumerator1, Coordinate Space=::World), Vector B=<0 0 1>))
       )
      )
      Y=&Lerp_Float(
       A=1
       B=&FN_FrontEndFloatModifier(
        Gameplay=&FloatFromCurve(
         FloatCurve=DataInterface() {propertyValues {Curve {keys [
             {
              interpMode=RCIM_Cubic
              leaveTangent=9.7835
              leaveTangentWeight=0.6125
              tangentMode=RCTM_Break
              tangentWeightMode=RCTWM_WeightedBoth
              time=0
              value=3.5431
             }
             {arriveTangent=-18.6565, interpMode=RCIM_Cubic, leaveTangent=-18.6565, time=0.1275, value=0.3703}
             {
              arriveTangent=42.0983
              arriveTangentWeight=0.0581
              interpMode=RCIM_Cubic
              leaveTangent=42.0983
              leaveTangentWeight=0.0669
              time=0.1303
              value=1.1118
             }
             {arriveTangent=4.0429, interpMode=RCIM_Cubic, leaveTangent=4.0429, time=0.2613, value=6}
             {arriveTangentWeight=0.0651, interpMode=RCIM_Cubic, leaveTangentWeight=0.0966, tangentMode=RCTM_User, time=0.35, value=2}
             {arriveTangentWeight=0.0581, interpMode=RCIM_Cubic, leaveTangentWeight=0.0623, tangentMode=RCTM_User, time=0.6, value=2.8}
             {arriveTangentWeight=0.0737, interpMode=RCIM_Cubic, tangentMode=RCTM_User, time=1, value=3}
            ]}}}
         CurveIndex=$Particles.PartitionParticles.Normalized Index In Partition
        )
        Front End=&FloatFromCurve(
         FloatCurve=DataInterface() {propertyValues {Curve {keys [
             {
              interpMode=RCIM_Cubic
              leaveTangent=9.7835
              leaveTangentWeight=0.6125
              tangentMode=RCTM_Break
              tangentWeightMode=RCTWM_WeightedBoth
              time=0
              value=3.5431
             }
             {arriveTangent=2.3923, interpMode=RCIM_Cubic, leaveTangent=2.3923, time=0.2613, value=6}
             {arriveTangentWeight=0.0651, interpMode=RCIM_Cubic, leaveTangentWeight=0.0966, tangentMode=RCTM_User, time=0.4, value=4.5}
             {arriveTangentWeight=0.0581, interpMode=RCIM_Cubic, leaveTangentWeight=0.0623, tangentMode=RCTM_User, time=0.6, value=2.8}
             {arriveTangentWeight=0.0737, interpMode=RCIM_Cubic, tangentMode=RCTM_User, time=1, value=3}
            ]}}}
         CurveIndex=$Particles.PartitionParticles.Normalized Index In Partition
        )
       )
       Alpha=&FloatFromCurve(
        FloatCurve=DataInterface() {propertyValues {Curve {keys [
            {
             arriveTangentWeight=0.616
             interpMode=RCIM_Cubic
             leaveTangentWeight=0.616
             tangentMode=RCTM_User
             tangentWeightMode=RCTWM_WeightedBoth
             time=0
             value=1
            }
            {
             arriveTangent=-3.3955
             arriveTangentWeight=0.688
             interpMode=RCIM_Cubic
             leaveTangent=-3.3955
             leaveTangentWeight=0.688
             tangentMode=RCTM_User
             tangentWeightMode=RCTWM_WeightedBoth
             time=1
             value=0
            }
           ]}}}
        CurveIndex=&Abs_Float(Float:NiagaraFloat=&DotProduct(
          Vector A=&ReturnMeshOrientationAxis(Mesh Orientation Axis=::ZAxis, Mesh Orientation Quat=$Particles.MeshOrientation)
          Vector B=&ReturnCameraVector(Camera Vectors=::NewEnumerator4, Coordinate Space=::Simulation, Origin Position=$Particles.Position)
         ))
       )
      )
     )
     Gameplay:Vector2f=&MakeVector2D(
      X=&Lerp_Float(
       A=&FloatFromCurve(
        FloatCurve=DataInterface() {propertyValues {Curve {keys [
            <at 0 1 RCIM_Cubic>
            {arriveTangentWeight=0.0507, interpMode=RCIM_Cubic, leaveTangentWeight=0.1376, tangentMode=RCTM_User, time=0.1193, value=2}
            {arriveTangentWeight=0.1467, interpMode=RCIM_Cubic, leaveTangentWeight=0.0257, tangentMode=RCTM_User, time=0.4433, value=2}
            {arriveTangentWeight=0.0257, interpMode=RCIM_Cubic, leaveTangentWeight=0.122, tangentMode=RCTM_User, time=0.5, value=1}
            <at 0.7693 1.05 RCIM_Cubic>
            <at 1 1 RCIM_Cubic>
           ]}}}
        CurveIndex=$Particles.PartitionParticles.Normalized Index In Partition
       )
       B=&FloatFromCurve(
        FloatCurve=DataInterface() {propertyValues {Curve {keys [
            <at 0 1 RCIM_Cubic>
            {arriveTangentWeight=0.0795, interpMode=RCIM_Cubic, leaveTangentWeight=0.1591, tangentMode=RCTM_User, time=0.1, value=0.5}
            {arriveTangentWeight=0.1272, interpMode=RCIM_Cubic, leaveTangentWeight=0.1272, tangentMode=RCTM_User, time=0.3, value=0.5}
            {arriveTangent=0.7143, interpMode=RCIM_Cubic, leaveTangent=0.7143, time=0.5, value=1}
            <at 1 1>
           ]}}}
        CurveIndex=$Particles.PartitionParticles.Normalized Index In Partition
       )
       Alpha=&FloatFromCurve(
        FloatCurve=DataInterface() {propertyValues {Curve {keys [<at 0.15 1>, {arriveTangent=1.8365, leaveTangent=1.8365, time=0.5, value=0}]}}}
        CurveIndex=&Abs_Float(Float:NiagaraFloat=&DotProduct(Vector A=&ReturnCameraVector(Camera Vectors=::NewEnumerator1, Coordinate Space=::World), Vector B=<0 0 1>))
       )
      )
      Y=&Lerp_Float(
       A=1
       B=&FloatFromCurve(
        FloatCurve=DataInterface() {propertyValues {Curve {keys [
            {
             interpMode=RCIM_Cubic
             leaveTangent=9.7835
             leaveTangentWeight=0.6125
             tangentMode=RCTM_Break
             tangentWeightMode=RCTWM_WeightedBoth
             time=0
             value=3.5431
            }
            {arriveTangent=-18.6565, interpMode=RCIM_Cubic, leaveTangent=-18.6565, time=0.1275, value=0.3703}
            {
             arriveTangent=42.0983
             arriveTangentWeight=0.0581
             interpMode=RCIM_Cubic
             leaveTangent=42.0983
             leaveTangentWeight=0.0669
             time=0.1303
             value=1.1118
            }
            {arriveTangent=4.0429, interpMode=RCIM_Cubic, leaveTangent=4.0429, time=0.2613, value=6}
            {arriveTangentWeight=0.0651, interpMode=RCIM_Cubic, leaveTangentWeight=0.0966, tangentMode=RCTM_User, time=0.35, value=2}
            {arriveTangentWeight=0.0581, interpMode=RCIM_Cubic, leaveTangentWeight=0.0623, tangentMode=RCTM_User, time=0.6, value=2.8}
            {arriveTangentWeight=0.0737, interpMode=RCIM_Cubic, tangentMode=RCTM_User, time=1, value=3}
           ]}}}
        CurveIndex=$Particles.PartitionParticles.Normalized Index In Partition
       )
       Alpha=&FloatFromCurve(
        FloatCurve=DataInterface() {propertyValues {Curve {keys [
            {
             arriveTangentWeight=0.616
             interpMode=RCIM_Cubic
             leaveTangentWeight=0.616
             tangentMode=RCTM_User
             tangentWeightMode=RCTWM_WeightedBoth
             time=0
             value=1
            }
            {
             arriveTangent=-3.3955
             arriveTangentWeight=0.688
             interpMode=RCIM_Cubic
             leaveTangent=-3.3955
             leaveTangentWeight=0.688
             tangentMode=RCTM_User
             tangentWeightMode=RCTWM_WeightedBoth
             time=1
             value=0
            }
           ]}}}
        CurveIndex=&Abs_Float(Float:NiagaraFloat=&DotProduct(
          Vector A=&ReturnMeshOrientationAxis(Mesh Orientation Axis=::ZAxis, Mesh Orientation Quat=$Particles.MeshOrientation)
          Vector B=&ReturnCameraVector(Camera Vectors=::NewEnumerator4, Coordinate Space=::Simulation, Origin Position=$Particles.Position)
         ))
       )
      )
     )
    )
    Scale Sprite Size Mode=::NewEnumerator2
   )
   &StaticMeshLocation(
    FixedBarycentric=false
    MaskSection=false
    MaskSpawnGroup=false
    OffsetAlongNormal=false
    SamplePosOffset=false
    Static Mesh:NiagaraDataInterfaceStaticMesh=$System.Static Mesh
    Triangle ID:NiagaraInt32=&ReturnParticleIDAttributes(Particle ID Attribute=::NewEnumerator1)
    Triangle Sampling Mode:ENiagara_StaticTriangleMode=::NewEnumerator2
   )
   &SolveForcesAndVelocity(Clamp Velocity=false, Limit Acceleration=false)
   &SpriteFacingAndAlignment(
    Sprite Alignment=&ReturnMeshOrientationAxis(Mesh Orientation Axis=::ZAxis, Mesh Orientation Quat=$Particles.MeshOrientation)
    Sprite Facing=&ReturnCameraVector(Camera Vectors=::NewEnumerator4, Coordinate Space=::Simulation, Origin Position=$Particles.Position)
    WriteAlignment=true
    WriteFacing=true
   )
   /SetVariables(Particles.Velocity=<1 0 0>) {isSetParameters=true, module=SetVariables_#5}
   &FN_Light_Attributes(
    Create Color Attribute=true
    Create Exponent Attribute=true
    Create Position Attribute=true
    Create Radius Attribute=true
    Enable Particle Light Enable=true
   )
   &ScaleColor(Scale RGB=&VectorFromFloat(Value=3), ScaleA=true, ScaleRGB=true)
   &ScaleColor(
    Scale Alpha=&FloatFromCurve(
     FloatCurve=DataInterface() {propertyValues {Curve {keys [
         {
          arriveTangent=-0.3625
          arriveTangentWeight=0.6683
          interpMode=RCIM_Cubic
          leaveTangent=-0.3625
          leaveTangentWeight=0.6683
          tangentMode=RCTM_User
          tangentWeightMode=RCTWM_WeightedBoth
          time=0
          value=1
         }
         {
          arriveTangent=-1.2454
          arriveTangentWeight=0.2827
          interpMode=RCIM_Cubic
          leaveTangent=-1.2454
          leaveTangentWeight=0.2827
          tangentMode=RCTM_User
          tangentWeightMode=RCTWM_WeightedBoth
          time=1
          value=0.1
         }
        ]}}}
     CurveIndex=&Abs_Float(Float:NiagaraFloat=&DotProduct(
       Vector A=&ReturnMeshOrientationAxis(Mesh Orientation Axis=::ZAxis, Mesh Orientation Quat=$Particles.MeshOrientation)
       Vector B=&ReturnCameraVector(Camera Vectors=::NewEnumerator4, Coordinate Space=::Simulation, Origin Position=$Particles.Position)
      ))
    )
    ScaleA=true
    ScaleRGB=true
   ) {module=ScaleColor001}
   &ScaleColor(
    Scale Alpha=&FN_FrontEndFloatModifier(
     Gameplay=&FloatFromCurve(
      FloatCurve=DataInterface() {propertyValues {Curve {keys [
          {arriveTangentWeight=0.0535, interpMode=RCIM_Cubic, leaveTangentWeight=0.107, tangentMode=RCTM_User, time=0.1899, value=1.1}
          {arriveTangentWeight=0.102, interpMode=RCIM_Cubic, leaveTangentWeight=0.2368, tangentMode=RCTM_User, time=0.1968, value=0.65}
          {arriveTangentWeight=0.0916, interpMode=RCIM_Cubic, leaveTangentWeight=0.0916, tangentMode=RCTM_User, time=0.3713, value=0.65}
          <at 0.4132 1.1>
         ]}}}
      CurveIndex=$Particles.PartitionParticles.Normalized Index In Partition
     )
     Front End=&FloatFromCurve(
      FloatCurve=DataInterface() {propertyValues {Curve {keys [
          {arriveTangentWeight=0.0535, interpMode=RCIM_Cubic, leaveTangentWeight=0.107, tangentMode=RCTM_User, time=0.1899, value=0.75}
          {arriveTangentWeight=0.102, interpMode=RCIM_Cubic, leaveTangentWeight=0.2368, tangentMode=RCTM_User, time=0.2, value=3}
          {arriveTangentWeight=0.0916, interpMode=RCIM_Cubic, leaveTangentWeight=0.0916, tangentMode=RCTM_User, time=0.4, value=3}
          <at 0.4132 1.1>
         ]}}}
      CurveIndex=$Particles.PartitionParticles.Normalized Index In Partition
     )
    )
    ScaleA=true
    ScaleRGB=true
   ) {module=ScaleColor002}
  ]
  rendererClasses [&NiagaraSpriteRendererProperties]
  renderers [{rendererClass=&NiagaraSpriteRendererProperties}]
 }
 Light {
  emitterUpdateScript [
   &EmitterState()
   &SpawnBurst_Instantaneous(Use Spawn Probability=false, UseLoopCountLimit=true)
  ]
  particleSpawnScript [&InitializeParticle(
    Color=<rgba 1 0.15 1 0.5>
    Lifetime=5
    Position Offset=<1 12 100>
    Sprite Size Mode=::NewEnumerator3
    UsePositionOffset=true
    Write Material Random=true
   )]
  particleUpdateScript [
   &ParticleState(Kill Particles When Lifetime Has Elapsed=false, Let Infinitely Lived Particles Die When Emitter Deactivates=true)
   &ScaleColor(Scale Alpha=15, ScaleA=true, ScaleRGB=true)
   /SetVariables(Particles.Position=$Particles.Initial.Position) {enabled=false, isSetParameters=true}
  ]
  rendererClasses [&NiagaraLightRendererProperties, &NiagaraSpriteRendererProperties]
  renderers [
   {rendererClass=&NiagaraLightRendererProperties, rendererIndex=0}
   {rendererClass=&NiagaraSpriteRendererProperties, rendererIndex=1}
  ]
 }
 Low {
  emitterSpawnScript [
   /SetVariables(Emitter.Num Particles To Spawn:NiagaraInt32=&self:Get_num_Tris(Static Mesh:NiagaraDataInterfaceStaticMesh=$System.Static Mesh)) {isSetParameters=true, module=SetVariables_#3}
   /SetVariables(Emitter.Plat Low=15) {isSetParameters=true, module=SetVariables_#4}
  ]
  emitterUpdateScript [
   &EmitterState()
   &SpawnBurst_Instantaneous(Spawn Count=$Emitter.Num Particles To Spawn, Use Spawn Probability=false, UseLoopCountLimit=true)
  ]
  particleSpawnScript [
   &InitializeParticle(
    Color Mode=::NewEnumerator0
    Lifetime=5
    Sprite Size Mode=::NewEnumerator3
    Uniform Sprite Size=8
    UsePositionOffset=false
    Write Material Random=true
   )
   &PartitionParticles(
    Normalized Index In Partition Includes One=true
    Number Of Partitions=1
    Particles Per Partition:NiagaraInt32=$Emitter.Num Particles To Spawn
    Partition Mode:ENiagara_PartitionMode=::NewEnumerator1
   )
   &Color(Color=&MultiplyLinearColor(
     Linear Color=&ColorFromCurve(
      ColorCurve=DataInterface() {propertyValues {
        AlphaCurve {keys [
          <at 0 0.1317>
          <at 0.0941 0.4021>
          <at 0.1935 0.093>
          <at 0.2965 0.093>
          {arriveTangentWeight=0.0471, interpMode=RCIM_Cubic, leaveTangentWeight=0.0162, tangentMode=RCTM_User, time=0.4002, value=0}
          {arriveTangentWeight=0.1704, interpMode=RCIM_Cubic, leaveTangentWeight=0.1641, tangentMode=RCTM_User, time=0.4426, value=0}
          <at 0.503 0.4663>
          {arriveTangent=-1.0579, interpMode=RCIM_Cubic, leaveTangent=-1.0579, time=0.6971, value=0.4663}
          <at 0.9438 0 RCIM_Cubic>
         ]}
        BlueCurve {keys [
          {interpMode=RCIM_Cubic, leaveTangentWeight=0.066, tangentMode=RCTM_User, time=0.3582, value=0.9242}
          {
           arriveTangent=-0.0196
           arriveTangentWeight=0.0526
           interpMode=RCIM_Cubic
           leaveTangent=-0.0196
           leaveTangentWeight=0.0879
           time=0.5128
           value=0.9162
          }
          {
           arriveTangent=0.0166
           arriveTangentWeight=0.0874
           interpMode=RCIM_Cubic
           leaveTangent=0.0166
           leaveTangentWeight=0.0774
           time=0.7712
           value=0.9162
          }
          {arriveTangentWeight=0.0763, interpMode=RCIM_Cubic, tangentMode=RCTM_User, time=1, value=0.9242}
         ]}
        GreenCurve {keys [
          {arriveTangent=-4.0724, interpMode=RCIM_Cubic, leaveTangent=-4.0724, leaveTangentWeight=0.066, tangentMode=RCTM_User, time=0.3582, value=0.2}
          {arriveTangentWeight=0.1478, interpMode=RCIM_Cubic, leaveTangentWeight=0.1152, tangentMode=RCTM_User, time=0.5128, value=0.1146}
          {arriveTangentWeight=0.1079, interpMode=RCIM_Cubic, leaveTangentWeight=0.1599, tangentMode=RCTM_User, time=0.7712, value=0.1146}
          {arriveTangent=4.0724, arriveTangentWeight=0.0763, interpMode=RCIM_Cubic, leaveTangent=4.0724, tangentMode=RCTM_User, time=1, value=0.2}
         ]}
        RedCurve {keys [
          {interpMode=RCIM_Cubic, leaveTangentWeight=0.066, tangentMode=RCTM_User, time=0.3582, value=1}
          {arriveTangentWeight=0.066, interpMode=RCIM_Cubic, leaveTangentWeight=0.0515, tangentMode=RCTM_User, time=0.5128, value=1}
          {arriveTangentWeight=0.0515, interpMode=RCIM_Cubic, leaveTangentWeight=0.0763, tangentMode=RCTM_User, time=0.7712, value=1}
          {arriveTangentWeight=0.0763, interpMode=RCIM_Cubic, tangentMode=RCTM_User, time=1, value=1}
         ]}
       }}
      CurveIndex=$Particles.PartitionParticles.Normalized Index In Partition
     )
     Scale Factor=&MakeFloat4FromVectorAndFloat("Vector (RGB)"=&VectorFromFloat(Value=3), "Float (Alpha)"=1)
    ))
   &KillParticles(Kill Particles=&SetBoolByIntComparison(
     A=&ModuloInt(
      A=&ReturnParticleIDAttributes(Particle ID Attribute=::NewEnumerator1)
      B=&MakeCustomIntFromBool(
       Bool=&IsPlatformSetActive(New Platform Set=DataInterface() {propertyValues {Platforms {qualityLevelMask=-31}}})
       True INT=$Emitter.Plat Low
       False INT=&MakeCustomIntFromBool(
        Bool=&IsPlatformSetActive(New Platform Set=DataInterface() {propertyValues {Platforms {qualityLevelMask=-30}}})
        True INT=$Emitter.PLat med
        False INT=&MakeCustomIntFromBool(
         Bool=&IsPlatformSetActive(New Platform Set=DataInterface() {propertyValues {Platforms {qualityLevelMask=-28}}})
         True INT=$Emitter.PLat High
         False INT=&MakeCustomIntFromBool(
          Bool=&IsPlatformSetActive(New Platform Set=DataInterface() {propertyValues {Platforms {qualityLevelMask=-8}}})
          True INT=$Emitter.Plat epicCin
          False INT=2147483647
         )
        )
       )
      )
     )
     B=0
     Comparison Type=::NewEnumerator5
    ))
   &KillParticles(Kill Particles=&Boolean_AndOperation(
     Bool1=&IsPlatformSetActive(New Platform Set=DataInterface() {propertyValues {Platforms {qualityLevelMask=-31}}})
     Bool2=&SetBoolByIntComparison(
      A=&ReturnParticleIDAttributes(Particle ID Attribute=::NewEnumerator1)
      B=0
      Comparison Type=::NewEnumerator4
     )
    )) {module=KillParticles001}
   &CameraOffset(Camera Offset Amount=7)
   &KillParticles(Kill Particles=&SetBoolByFloatComparison(
     A=&MakeFloatFromLinearColor(LinearColor=$Particles.Color, Channel:ENiagaraLinearColor_Channels=::NewEnumerator3)
     B=0
     Comparison Type=::NewEnumerator7
    )) {module=KillParticles002}
  ]
  particleUpdateScript [
   &ParticleState(Kill Particles When Lifetime Has Elapsed=false, Let Infinitely Lived Particles Die When Emitter Deactivates=true)
   &ScaleSpriteSize(
    Scale Sprite Size Mode=::NewEnumerator0
    Uniform Scale Factor=&FloatFromCurve(
     FloatCurve=DataInterface() {propertyValues {Curve {keys [
         {arriveTangent=3.4855, interpMode=RCIM_Cubic, leaveTangent=3.4855, leaveTangentWeight=0.1686, tangentMode=RCTM_User, time=0, value=0.65}
         {
          arriveTangent=0.6084
          arriveTangentWeight=0.0544
          interpMode=RCIM_Cubic
          leaveTangentWeight=0.0927
          tangentMode=RCTM_Break
          time=0.1394
          value=1
         }
         {
          arriveTangentWeight=0.0927
          interpMode=RCIM_Cubic
          leaveTangent=5.4842
          leaveTangentWeight=0.3238
          tangentMode=RCTM_Break
          time=0.4176
          value=1
         }
         {arriveTangentWeight=0.0581, interpMode=RCIM_Cubic, leaveTangentWeight=0.1361, tangentMode=RCTM_User, time=0.5918, value=1.6527}
         {arriveTangent=-1.5358, arriveTangentWeight=0.1361, interpMode=RCIM_Cubic, leaveTangent=-1.5358, tangentMode=RCTM_User, time=1, value=1}
        ]}}}
     CurveIndex=$Particles.PartitionParticles.Normalized Index In Partition
    )
   ) {module=ScaleSpriteSize001}
   &ScaleSpriteSize(
    Scale Factor=&FN_FrontEndVector2Modifier(
     Front End:Vector2f=&MakeVector2D(
      X=&Lerp_Float(
       A=&FloatFromCurve(
        FloatCurve=DataInterface() {propertyValues {Curve {keys [
            <at 0.0958 1 RCIM_Cubic>
            {arriveTangent=6.9016, interpMode=RCIM_Cubic, leaveTangent=6.9016, time=0.1, value=3}
            {arriveTangent=-5, interpMode=RCIM_Cubic, leaveTangent=-5, time=0.3856, value=3}
            {arriveTangent=-5.082, interpMode=RCIM_Cubic, leaveTangent=-5.082, time=0.5, value=1}
            <at 0.7693 1.05 RCIM_Cubic>
            <at 1 1 RCIM_Cubic>
           ]}}}
        CurveIndex=$Particles.PartitionParticles.Normalized Index In Partition
       )
       B=&FloatFromCurve(
        FloatCurve=DataInterface() {propertyValues {Curve {keys [
            <at 0 1 RCIM_Cubic>
            {arriveTangent=7.5, interpMode=RCIM_Cubic, leaveTangent=7.5, time=0.1, value=1.25}
            {arriveTangent=-0.625, interpMode=RCIM_Cubic, leaveTangent=-0.625, time=0.4, value=4}
            {arriveTangent=-5, interpMode=RCIM_Cubic, leaveTangent=-5, time=0.5, value=1}
            <at 1 1>
           ]}}}
        CurveIndex=$Particles.PartitionParticles.Normalized Index In Partition
       )
       Alpha=&Abs_Float(Float:NiagaraFloat=&DotProduct(
         Vector A=&ReturnCameraVector(Camera Vectors=::NewEnumerator1, Coordinate Space=::Simulation)
         Vector B=$Engine.Owner.SystemXAxis
        ))
      )
      Y=&Lerp_Float(
       A=1
       B=&FloatFromCurve(
        FloatCurve=DataInterface() {propertyValues {Curve {keys [
            {
             interpMode=RCIM_Cubic
             leaveTangent=9.7835
             leaveTangentWeight=0.6125
             tangentMode=RCTM_Break
             tangentWeightMode=RCTWM_WeightedBoth
             time=0.0087
             value=6.5905
            }
            {arriveTangent=-1.9961, interpMode=RCIM_Cubic, leaveTangent=-1.9961, tangentMode=RCTM_Break, time=0.1493, value=3}
            {arriveTangent=45.0868, interpMode=RCIM_Cubic, leaveTangent=45.0868, time=0.2, value=10}
            {arriveTangent=-65.1771, interpMode=RCIM_Cubic, leaveTangent=-65.1771, time=0.3045, value=10}
            {arriveTangentWeight=0.0651, interpMode=RCIM_Cubic, leaveTangentWeight=0.0966, tangentMode=RCTM_User, time=0.3217, value=2.0662}
            {arriveTangentWeight=0.0581, interpMode=RCIM_Cubic, leaveTangentWeight=0.0623, tangentMode=RCTM_User, time=0.6, value=2.8}
            {arriveTangentWeight=0.0737, interpMode=RCIM_Cubic, tangentMode=RCTM_User, time=1, value=3}
           ]}}}
        CurveIndex=$Particles.PartitionParticles.Normalized Index In Partition
        Scale Curve=1.5
       )
       Alpha=&FloatFromCurve(
        FloatCurve=DataInterface() {propertyValues {Curve {keys [
            {
             arriveTangentWeight=0.616
             interpMode=RCIM_Cubic
             leaveTangentWeight=0.616
             tangentMode=RCTM_User
             tangentWeightMode=RCTWM_WeightedBoth
             time=0
             value=1
            }
            {
             arriveTangent=-3.3955
             arriveTangentWeight=0.688
             interpMode=RCIM_Cubic
             leaveTangent=-3.3955
             leaveTangentWeight=0.688
             tangentMode=RCTM_User
             tangentWeightMode=RCTWM_WeightedBoth
             time=1
             value=0
            }
           ]}}}
        CurveIndex=&Abs_Float(Float:NiagaraFloat=&DotProduct(
          Vector A=&ReturnMeshOrientationAxis(Mesh Orientation Axis=::ZAxis, Mesh Orientation Quat=$Particles.MeshOrientation)
          Vector B=&ReturnCameraVector(Camera Vectors=::NewEnumerator4, Coordinate Space=::Simulation, Origin Position=$Particles.Position)
         ))
       )
      )
     )
     Gameplay:Vector2f=&MakeVector2D(
      X=&Lerp_Float(
       A=&FloatFromCurve(
        FloatCurve=DataInterface() {propertyValues {Curve {keys [
            <at 0.0958 1 RCIM_Cubic>
            {arriveTangent=3.0634, interpMode=RCIM_Cubic, leaveTangent=3.0634, time=0.1, value=1.8877}
            {arriveTangent=-2.2194, interpMode=RCIM_Cubic, leaveTangent=-2.2194, time=0.3856, value=1.8877}
            {arriveTangent=-2.1833, interpMode=RCIM_Cubic, leaveTangent=-2.1833, time=0.5, value=1}
            <at 0.7693 1.05 RCIM_Cubic>
            <at 1 1 RCIM_Cubic>
           ]}}}
        CurveIndex=$Particles.PartitionParticles.Normalized Index In Partition
       )
       B=&FloatFromCurve(
        FloatCurve=DataInterface() {propertyValues {Curve {keys [
            <at 0 1 RCIM_Cubic>
            {arriveTangent=2.875, interpMode=RCIM_Cubic, leaveTangent=2.875, time=0.1, value=1.25}
            {arriveTangent=-0.625, interpMode=RCIM_Cubic, leaveTangent=-0.625, time=0.4, value=2.15}
            {arriveTangent=-1.9167, interpMode=RCIM_Cubic, leaveTangent=-1.9167, time=0.5, value=1}
            <at 1 1>
           ]}}}
        CurveIndex=$Particles.PartitionParticles.Normalized Index In Partition
       )
       Alpha=&Abs_Float(Float:NiagaraFloat=&DotProduct(
         Vector A=&ReturnCameraVector(Camera Vectors=::NewEnumerator1, Coordinate Space=::Simulation)
         Vector B=$Engine.Owner.SystemXAxis
        ))
      )
      Y=&Lerp_Float(
       A=1
       B=&FloatFromCurve(
        FloatCurve=DataInterface() {propertyValues {Curve {keys [
            {
             interpMode=RCIM_Cubic
             leaveTangent=9.7835
             leaveTangentWeight=0.6125
             tangentMode=RCTM_Break
             tangentWeightMode=RCTWM_WeightedBoth
             time=0.0087
             value=6.5905
            }
            {arriveTangent=-1.9961, interpMode=RCIM_Cubic, leaveTangent=-1.9961, time=0.1493, value=1.6953}
            {arriveTangent=2.1513, interpMode=RCIM_Cubic, leaveTangent=2.1513, time=0.3045, value=6}
            {arriveTangentWeight=0.0651, interpMode=RCIM_Cubic, leaveTangentWeight=0.0966, tangentMode=RCTM_User, time=0.3217, value=2.0662}
            {arriveTangentWeight=0.0581, interpMode=RCIM_Cubic, leaveTangentWeight=0.0623, tangentMode=RCTM_User, time=0.6, value=2.8}
            {arriveTangentWeight=0.0737, interpMode=RCIM_Cubic, tangentMode=RCTM_User, time=1, value=3}
           ]}}}
        CurveIndex=$Particles.PartitionParticles.Normalized Index In Partition
        Scale Curve=1.5
       )
       Alpha=&FloatFromCurve(
        FloatCurve=DataInterface() {propertyValues {Curve {keys [
            {
             arriveTangentWeight=0.616
             interpMode=RCIM_Cubic
             leaveTangentWeight=0.616
             tangentMode=RCTM_User
             tangentWeightMode=RCTWM_WeightedBoth
             time=0
             value=1
            }
            {
             arriveTangent=-3.3955
             arriveTangentWeight=0.688
             interpMode=RCIM_Cubic
             leaveTangent=-3.3955
             leaveTangentWeight=0.688
             tangentMode=RCTM_User
             tangentWeightMode=RCTWM_WeightedBoth
             time=1
             value=0
            }
           ]}}}
        CurveIndex=&Abs_Float(Float:NiagaraFloat=&DotProduct(
          Vector A=&ReturnMeshOrientationAxis(Mesh Orientation Axis=::ZAxis, Mesh Orientation Quat=$Particles.MeshOrientation)
          Vector B=&ReturnCameraVector(Camera Vectors=::NewEnumerator4, Coordinate Space=::Simulation, Origin Position=$Particles.Position)
         ))
       )
      )
     )
    )
    Scale Sprite Size Mode=::NewEnumerator2
   )
   &StaticMeshLocation(
    FixedBarycentric=false
    MaskSection=false
    MaskSpawnGroup=false
    OffsetAlongNormal=false
    SamplePosOffset=false
    Static Mesh:NiagaraDataInterfaceStaticMesh=$System.Static Mesh
    Triangle ID:NiagaraInt32=&ReturnParticleIDAttributes(Particle ID Attribute=::NewEnumerator1)
    Triangle Sampling Mode:ENiagara_StaticTriangleMode=::NewEnumerator2
   )
   &SolveForcesAndVelocity(Clamp Velocity=false, Limit Acceleration=false)
   &SpriteFacingAndAlignment(
    Sprite Alignment=&ReturnMeshOrientationAxis(Mesh Orientation Axis=::ZAxis, Mesh Orientation Quat=$Particles.MeshOrientation)
    Sprite Facing=&ReturnCameraVector(Camera Vectors=::NewEnumerator4, Coordinate Space=::Simulation, Origin Position=$Particles.Position)
    WriteAlignment=true
    WriteFacing=true
   )
   /SetVariables(Particles.Velocity=<1 0 0>) {isSetParameters=true, module=SetVariables_#5}
   &FN_Light_Attributes(
    Create Color Attribute=true
    Create Exponent Attribute=true
    Create Position Attribute=true
    Create Radius Attribute=true
    Enable Particle Light Enable=true
   )
   &ScaleColor(Scale RGB=&VectorFromFloat(Value=3), ScaleA=true, ScaleRGB=true)
   &ScaleColor(
    Scale Alpha=&FloatFromCurve(
     FloatCurve=DataInterface() {propertyValues {Curve {keys [
         {
          arriveTangent=-0.3625
          arriveTangentWeight=0.6683
          interpMode=RCIM_Cubic
          leaveTangent=-0.3625
          leaveTangentWeight=0.6683
          tangentMode=RCTM_User
          tangentWeightMode=RCTWM_WeightedBoth
          time=0
          value=1
         }
         {
          arriveTangent=-1.2454
          arriveTangentWeight=0.2827
          interpMode=RCIM_Cubic
          leaveTangent=-1.2454
          leaveTangentWeight=0.2827
          tangentMode=RCTM_User
          tangentWeightMode=RCTWM_WeightedBoth
          time=1
          value=0.1
         }
        ]}}}
     CurveIndex=&Abs_Float(Float:NiagaraFloat=&DotProduct(
       Vector A=&ReturnMeshOrientationAxis(Mesh Orientation Axis=::ZAxis, Mesh Orientation Quat=$Particles.MeshOrientation)
       Vector B=&ReturnCameraVector(Camera Vectors=::NewEnumerator4, Coordinate Space=::Simulation, Origin Position=$Particles.Position)
      ))
    )
    ScaleA=true
    ScaleRGB=true
   ) {module=ScaleColor001}
   &ScaleColor(
    Scale Alpha=&FN_FrontEndFloatModifier(
     Gameplay=&FloatFromCurve(
      FloatCurve=DataInterface() {propertyValues {Curve {keys [
          <at 0.0515 1 RCIM_Cubic>
          {arriveTangentWeight=0.0535, interpMode=RCIM_Cubic, leaveTangentWeight=0.107, tangentMode=RCTM_User, time=0.1515, value=0.4}
          {arriveTangentWeight=0.0916, interpMode=RCIM_Cubic, leaveTangentWeight=0.0916, tangentMode=RCTM_User, time=0.3, value=0.4}
          <at 0.5 1>
          <at 1 1>
         ]}}}
      CurveIndex=$Particles.PartitionParticles.Normalized Index In Partition
     )
     Front End=&FloatFromCurve(
      FloatCurve=DataInterface() {propertyValues {Curve {keys [
          <at 0.0515 1 RCIM_Cubic>
          {arriveTangentWeight=0.0535, interpMode=RCIM_Cubic, leaveTangentWeight=0.107, tangentMode=RCTM_User, time=0.1515, value=1}
          {arriveTangentWeight=0.0916, interpMode=RCIM_Cubic, leaveTangentWeight=0.0916, tangentMode=RCTM_User, time=0.3, value=2}
          <at 0.5 1>
          <at 1 1>
         ]}}}
      CurveIndex=$Particles.PartitionParticles.Normalized Index In Partition
     )
    )
    ScaleA=true
    ScaleRGB=true
   ) {module=ScaleColor002}
  ]
  rendererClasses [&NiagaraSpriteRendererProperties]
  renderers [{rendererClass=&NiagaraSpriteRendererProperties}]
 }
 Med {
  emitterSpawnScript [
   /SetVariables(Emitter.Num Particles To Spawn:NiagaraInt32=&self:Get_num_Tris(Static Mesh:NiagaraDataInterfaceStaticMesh=$System.Static Mesh)) {isSetParameters=true, module=SetVariables_#3}
   /SetVariables(Emitter.PLat med=7) {isSetParameters=true, module=SetVariables_#4}
  ]
  emitterUpdateScript [
   &EmitterState()
   &SpawnBurst_Instantaneous(Spawn Count=$Emitter.Num Particles To Spawn, Use Spawn Probability=false, UseLoopCountLimit=true)
  ]
  particleSpawnScript [
   &InitializeParticle(
    Color Mode=::NewEnumerator0
    Lifetime=5
    Sprite Size Mode=::NewEnumerator3
    Uniform Sprite Size=8
    UsePositionOffset=false
    Write Material Random=true
   )
   &PartitionParticles(
    Normalized Index In Partition Includes One=true
    Number Of Partitions=1
    Particles Per Partition:NiagaraInt32=$Emitter.Num Particles To Spawn
    Partition Mode:ENiagara_PartitionMode=::NewEnumerator1
   )
   &Color(Color=&MultiplyLinearColor(
     Linear Color=&ColorFromCurve(
      ColorCurve=DataInterface() {propertyValues {
        AlphaCurve {keys [
          <at 0 0.1317>
          <at 0.0941 0.4021>
          <at 0.1935 0.093>
          <at 0.2965 0.093>
          {arriveTangentWeight=0.0471, interpMode=RCIM_Cubic, leaveTangentWeight=0.0162, tangentMode=RCTM_User, time=0.4002, value=0}
          {arriveTangentWeight=0.1704, interpMode=RCIM_Cubic, leaveTangentWeight=0.1641, tangentMode=RCTM_User, time=0.4426, value=0}
          <at 0.503 0.4663>
          {arriveTangent=-1.0579, interpMode=RCIM_Cubic, leaveTangent=-1.0579, time=0.6971, value=0.4663}
          <at 0.9438 0 RCIM_Cubic>
         ]}
        BlueCurve {keys [
          {interpMode=RCIM_Cubic, leaveTangentWeight=0.066, tangentMode=RCTM_User, time=0.3582, value=0.9242}
          {
           arriveTangent=-0.0196
           arriveTangentWeight=0.0526
           interpMode=RCIM_Cubic
           leaveTangent=-0.0196
           leaveTangentWeight=0.0879
           time=0.5128
           value=0.9162
          }
          {
           arriveTangent=0.0166
           arriveTangentWeight=0.0874
           interpMode=RCIM_Cubic
           leaveTangent=0.0166
           leaveTangentWeight=0.0774
           time=0.7712
           value=0.9162
          }
          {arriveTangentWeight=0.0763, interpMode=RCIM_Cubic, tangentMode=RCTM_User, time=1, value=0.9242}
         ]}
        GreenCurve {keys [
          {arriveTangent=-4.0724, interpMode=RCIM_Cubic, leaveTangent=-4.0724, leaveTangentWeight=0.066, tangentMode=RCTM_User, time=0.3582, value=0.2}
          {arriveTangentWeight=0.1478, interpMode=RCIM_Cubic, leaveTangentWeight=0.1152, tangentMode=RCTM_User, time=0.5128, value=0.1146}
          {arriveTangentWeight=0.1079, interpMode=RCIM_Cubic, leaveTangentWeight=0.1599, tangentMode=RCTM_User, time=0.7712, value=0.1146}
          {arriveTangent=4.0724, arriveTangentWeight=0.0763, interpMode=RCIM_Cubic, leaveTangent=4.0724, tangentMode=RCTM_User, time=1, value=0.2}
         ]}
        RedCurve {keys [
          {interpMode=RCIM_Cubic, leaveTangentWeight=0.066, tangentMode=RCTM_User, time=0.3582, value=1}
          {arriveTangentWeight=0.066, interpMode=RCIM_Cubic, leaveTangentWeight=0.0515, tangentMode=RCTM_User, time=0.5128, value=1}
          {arriveTangentWeight=0.0515, interpMode=RCIM_Cubic, leaveTangentWeight=0.0763, tangentMode=RCTM_User, time=0.7712, value=1}
          {arriveTangentWeight=0.0763, interpMode=RCIM_Cubic, tangentMode=RCTM_User, time=1, value=1}
         ]}
       }}
      CurveIndex=$Particles.PartitionParticles.Normalized Index In Partition
     )
     Scale Factor=&MakeFloat4FromVectorAndFloat("Vector (RGB)"=&VectorFromFloat(Value=5), "Float (Alpha)"=0.4)
    ))
   &KillParticles(Kill Particles=&SetBoolByIntComparison(
     A=&ModuloInt(
      A=&ReturnParticleIDAttributes(Particle ID Attribute=::NewEnumerator1)
      B=&MakeCustomIntFromBool(
       Bool=&IsPlatformSetActive(New Platform Set=DataInterface() {propertyValues {Platforms {qualityLevelMask=-31}}})
       True INT=$Emitter.Plat Low
       False INT=&MakeCustomIntFromBool(
        Bool=&IsPlatformSetActive(New Platform Set=DataInterface() {propertyValues {Platforms {qualityLevelMask=-30}}})
        True INT=$Emitter.PLat med
        False INT=&MakeCustomIntFromBool(
         Bool=&IsPlatformSetActive(New Platform Set=DataInterface() {propertyValues {Platforms {qualityLevelMask=-28}}})
         True INT=$Emitter.PLat High
         False INT=&MakeCustomIntFromBool(
          Bool=&IsPlatformSetActive(New Platform Set=DataInterface() {propertyValues {Platforms {qualityLevelMask=-8}}})
          True INT=$Emitter.Plat epicCin
          False INT=2147483647
         )
        )
       )
      )
     )
     B=0
     Comparison Type=::NewEnumerator5
    ))
   &KillParticles(Kill Particles=&Boolean_AndOperation(
     Bool1=&IsPlatformSetActive(New Platform Set=DataInterface() {propertyValues {Platforms {qualityLevelMask=-31}}})
     Bool2=&SetBoolByIntComparison(
      A=&ReturnParticleIDAttributes(Particle ID Attribute=::NewEnumerator1)
      B=0
      Comparison Type=::NewEnumerator4
     )
    )) {module=KillParticles001}
   &CameraOffset(Camera Offset Amount=7)
   &KillParticles(Kill Particles=&SetBoolByFloatComparison(
     A=&MakeFloatFromLinearColor(LinearColor=$Particles.Color, Channel:ENiagaraLinearColor_Channels=::NewEnumerator3)
     B=0
     Comparison Type=::NewEnumerator7
    )) {module=KillParticles002}
  ]
  particleUpdateScript [
   &ParticleState(Kill Particles When Lifetime Has Elapsed=false, Let Infinitely Lived Particles Die When Emitter Deactivates=true)
   &ScaleSpriteSize(
    Scale Sprite Size Mode=::NewEnumerator0
    Uniform Scale Factor=&FloatFromCurve(
     FloatCurve=DataInterface() {propertyValues {Curve {keys [
         {arriveTangent=3.4855, interpMode=RCIM_Cubic, leaveTangent=3.4855, leaveTangentWeight=0.1686, tangentMode=RCTM_User, time=0, value=0.65}
         {
          arriveTangent=0.6084
          arriveTangentWeight=0.0544
          interpMode=RCIM_Cubic
          leaveTangentWeight=0.0927
          tangentMode=RCTM_Break
          time=0.1394
          value=1
         }
         {
          arriveTangentWeight=0.0927
          interpMode=RCIM_Cubic
          leaveTangent=5.4842
          leaveTangentWeight=0.3238
          tangentMode=RCTM_Break
          time=0.4176
          value=1
         }
         {arriveTangentWeight=0.0581, interpMode=RCIM_Cubic, leaveTangentWeight=0.1361, tangentMode=RCTM_User, time=0.5918, value=1.6527}
         {arriveTangent=-1.5358, arriveTangentWeight=0.1361, interpMode=RCIM_Cubic, leaveTangent=-1.5358, tangentMode=RCTM_User, time=1, value=1}
        ]}}}
     CurveIndex=$Particles.PartitionParticles.Normalized Index In Partition
    )
   ) {module=ScaleSpriteSize001}
   &ScaleSpriteSize(
    Scale Factor=&MakeVector2D(
     X=&Lerp_Float(
      A=&FloatFromCurve(
       FloatCurve=DataInterface() {propertyValues {Curve {keys [
           <at 0.137 1 RCIM_Cubic>
           {arriveTangent=6.1348, interpMode=RCIM_Cubic, leaveTangent=6.1348, time=0.1676, value=2}
           {arriveTangent=-3.0087, interpMode=RCIM_Cubic, leaveTangent=-3.0087, time=0.3, value=2}
           {arriveTangent=-2.0241, interpMode=RCIM_Cubic, leaveTangent=-2.0241, time=0.5, value=1}
           <at 0.7693 1.05 RCIM_Cubic>
           <at 1 1 RCIM_Cubic>
          ]}}}
       CurveIndex=$Particles.PartitionParticles.Normalized Index In Partition
      )
      B=&FloatFromCurve(
       FloatCurve=DataInterface() {propertyValues {Curve {keys [
           <at 0 1 RCIM_Cubic>
           {arriveTangent=3.5, interpMode=RCIM_Cubic, leaveTangent=3.5, time=0.1, value=2.05}
           {arriveTangent=-2.625, interpMode=RCIM_Cubic, leaveTangent=-2.625, time=0.3, value=2.05}
           {arriveTangent=-1.5, interpMode=RCIM_Cubic, leaveTangent=-1.5, time=0.5, value=1}
           <at 1 1>
          ]}}}
       CurveIndex=$Particles.PartitionParticles.Normalized Index In Partition
      )
      Alpha=&Abs_Float(Float:NiagaraFloat=&DotProduct(
        Vector A=&ReturnCameraVector(Camera Vectors=::NewEnumerator1, Coordinate Space=::Simulation)
        Vector B=$Engine.Owner.SystemXAxis
       ))
     )
     Y=&Lerp_Float(
      A=1
      B=&FloatFromCurve(
       FloatCurve=DataInterface() {propertyValues {Curve {keys [
           {
            interpMode=RCIM_Cubic
            leaveTangent=9.7835
            leaveTangentWeight=0.6125
            tangentMode=RCTM_Break
            tangentWeightMode=RCTWM_WeightedBoth
            time=0
            value=3.5431
           }
           {
            arriveTangent=9.4043
            arriveTangentWeight=0.0581
            interpMode=RCIM_Cubic
            leaveTangent=9.4043
            leaveTangentWeight=0.0669
            time=0.1
            value=2.375
           }
           {arriveTangent=-1.5, interpMode=RCIM_Cubic, leaveTangent=-1.5, time=0.2613, value=6}
           {arriveTangentWeight=0.0651, interpMode=RCIM_Cubic, leaveTangentWeight=0.0966, tangentMode=RCTM_User, time=0.35, value=2}
           {arriveTangentWeight=0.0581, interpMode=RCIM_Cubic, leaveTangentWeight=0.0623, tangentMode=RCTM_User, time=0.6, value=2.8}
           {arriveTangentWeight=0.0737, interpMode=RCIM_Cubic, tangentMode=RCTM_User, time=1, value=3}
          ]}}}
       CurveIndex=$Particles.PartitionParticles.Normalized Index In Partition
      )
      Alpha=&FloatFromCurve(
       FloatCurve=DataInterface() {propertyValues {Curve {keys [
           {
            arriveTangentWeight=0.616
            interpMode=RCIM_Cubic
            leaveTangentWeight=0.616
            tangentMode=RCTM_User
            tangentWeightMode=RCTWM_WeightedBoth
            time=0
            value=1
           }
           {
            arriveTangent=-3.3955
            arriveTangentWeight=0.688
            interpMode=RCIM_Cubic
            leaveTangent=-3.3955
            leaveTangentWeight=0.688
            tangentMode=RCTM_User
            tangentWeightMode=RCTWM_WeightedBoth
            time=1
            value=0
           }
          ]}}}
       CurveIndex=&Abs_Float(Float:NiagaraFloat=&DotProduct(
         Vector A=&ReturnMeshOrientationAxis(Mesh Orientation Axis=::ZAxis, Mesh Orientation Quat=$Particles.MeshOrientation)
         Vector B=&ReturnCameraVector(Camera Vectors=::NewEnumerator4, Coordinate Space=::Simulation, Origin Position=$Particles.Position)
        ))
      )
     )
    )
    Scale Sprite Size Mode=::NewEnumerator2
   )
   &StaticMeshLocation(
    FixedBarycentric=false
    MaskSection=false
    MaskSpawnGroup=false
    OffsetAlongNormal=false
    SamplePosOffset=false
    Static Mesh:NiagaraDataInterfaceStaticMesh=$System.Static Mesh
    Triangle ID:NiagaraInt32=&ReturnParticleIDAttributes(Particle ID Attribute=::NewEnumerator1)
    Triangle Sampling Mode:ENiagara_StaticTriangleMode=::NewEnumerator2
   )
   &SolveForcesAndVelocity(Clamp Velocity=false, Limit Acceleration=false)
   &SpriteFacingAndAlignment(
    Sprite Alignment=&ReturnMeshOrientationAxis(Mesh Orientation Axis=::ZAxis, Mesh Orientation Quat=$Particles.MeshOrientation)
    Sprite Facing=&ReturnCameraVector(Camera Vectors=::NewEnumerator4, Coordinate Space=::Simulation, Origin Position=$Particles.Position)
    WriteAlignment=true
    WriteFacing=true
   )
   /SetVariables(Particles.Velocity=<1 0 0>) {isSetParameters=true, module=SetVariables_#5}
   &FN_Light_Attributes(
    Create Color Attribute=true
    Create Exponent Attribute=true
    Create Position Attribute=true
    Create Radius Attribute=true
    Enable Particle Light Enable=true
   )
   &ScaleColor(Scale RGB=&VectorFromFloat(Value=3), ScaleA=true, ScaleRGB=true)
   &ScaleColor(
    Scale Alpha=&FloatFromCurve(
     FloatCurve=DataInterface() {propertyValues {Curve {keys [
         {
          arriveTangent=-0.3625
          arriveTangentWeight=0.6683
          interpMode=RCIM_Cubic
          leaveTangent=-0.3625
          leaveTangentWeight=0.6683
          tangentMode=RCTM_User
          tangentWeightMode=RCTWM_WeightedBoth
          time=0
          value=1
         }
         {
          arriveTangent=-1.2454
          arriveTangentWeight=0.2827
          interpMode=RCIM_Cubic
          leaveTangent=-1.2454
          leaveTangentWeight=0.2827
          tangentMode=RCTM_User
          tangentWeightMode=RCTWM_WeightedBoth
          time=1
          value=0.1
         }
        ]}}}
     CurveIndex=&Abs_Float(Float:NiagaraFloat=&DotProduct(
       Vector A=&ReturnMeshOrientationAxis(Mesh Orientation Axis=::ZAxis, Mesh Orientation Quat=$Particles.MeshOrientation)
       Vector B=&ReturnCameraVector(Camera Vectors=::NewEnumerator4, Coordinate Space=::Simulation, Origin Position=$Particles.Position)
      ))
    )
    ScaleA=true
    ScaleRGB=true
   ) {module=ScaleColor001}
   &ScaleColor(
    Scale Alpha=&FN_FrontEndFloatModifier(
     Gameplay=&FloatFromCurve(
      FloatCurve=DataInterface() {propertyValues {Curve {keys [
          {arriveTangentWeight=0.0535, interpMode=RCIM_Cubic, leaveTangentWeight=0.107, tangentMode=RCTM_User, time=0.1899, value=1.1}
          {arriveTangentWeight=0.102, interpMode=RCIM_Cubic, leaveTangentWeight=0.2368, tangentMode=RCTM_User, time=0.1968, value=0.65}
          {arriveTangentWeight=0.0916, interpMode=RCIM_Cubic, leaveTangentWeight=0.0916, tangentMode=RCTM_User, time=0.3713, value=0.65}
          <at 0.4132 1.1>
         ]}}}
      CurveIndex=$Particles.PartitionParticles.Normalized Index In Partition
     )
     Front End=&FloatFromCurve(
      FloatCurve=DataInterface() {propertyValues {Curve {keys [
          {arriveTangentWeight=0.0535, interpMode=RCIM_Cubic, leaveTangentWeight=0.107, tangentMode=RCTM_User, time=0.1899, value=0.75}
          {arriveTangentWeight=0.102, interpMode=RCIM_Cubic, leaveTangentWeight=0.2368, tangentMode=RCTM_User, time=0.2, value=2.5}
          {arriveTangentWeight=0.0916, interpMode=RCIM_Cubic, leaveTangentWeight=0.0916, tangentMode=RCTM_User, time=0.4, value=2.5}
          <at 0.4132 1.1>
         ]}}}
      CurveIndex=$Particles.PartitionParticles.Normalized Index In Partition
     )
    )
    ScaleA=true
    ScaleRGB=true
   ) {module=ScaleColor002}
  ]
  rendererClasses [&NiagaraSpriteRendererProperties]
  renderers [{rendererClass=&NiagaraSpriteRendererProperties}]
 }
 debug {
  bEnabled=false
  emitterUpdateScript [
   &EmitterState()
   &SpawnBurst_Instantaneous(Use Spawn Probability=false, UseLoopCountLimit=false)
  ]
  particleSpawnScript [
   &InitializeParticle(Color Mode=::NewEnumerator0, Lifetime=5, UsePositionOffset=false, Write Material Random=true)
   &CameraOffset(Camera Offset Amount=25)
  ]
  particleUpdateScript [
   &ParticleState()
   &DynamicMaterialParameters(
    Index 0 Param 1=&FloatFromCurve(
     FloatCurve=DataInterface() {propertyValues {Curve {keys [<at -1 -1>, {arriveTangent=1.8365, leaveTangent=1.8365, time=1, value=1}]}}}
     CurveIndex=&Abs_Float(Float:NiagaraFloat=&DotProduct(Vector A=&ReturnCameraVector(Camera Vectors=::NewEnumerator1, Coordinate Space=::World), Vector B=<0 0 1>))
    )
    Index 0 Param 3=&DotProduct(
     Vector A=&ReturnCameraVector(Camera Vectors=::NewEnumerator1, Coordinate Space=::Simulation)
     Vector B=$Engine.Owner.SystemXAxis
    )
    Index 0 Param 4=&Abs_Float(Float:NiagaraFloat=&DotProduct(
      Vector A=&ReturnCameraVector(Camera Vectors=::NewEnumerator1, Coordinate Space=::World)
      Vector B=&TransformVector(Vector=$Engine.Owner.SystemXAxis, Source Space=::Simulation, Destination Space=::World)
     ))
    Module.Index0Param0Write=true
    Module.Index0Param1Write=true
    Module.Index0Param2Write=true
    Module.Index0Param3Write=true
   )
  ]
  rendererClasses [&NiagaraSpriteRendererProperties, &NiagaraMeshRendererProperties]
  renderers [
   {rendererClass=&NiagaraSpriteRendererProperties}
   {rendererClass=&NiagaraMeshRendererProperties, rendererIndex=1}
  ]
 }
}
summary {
 systemName=NS_FelineWarrior_Pick_Idle
 emitters [
  {emitterName=EpicCine, rendererClasses [&NiagaraSpriteRendererProperties]}
  {emitterName=High, rendererClasses [&NiagaraSpriteRendererProperties]}
  {emitterName=Med, rendererClasses [&NiagaraSpriteRendererProperties]}
  {emitterName=Low, rendererClasses [&NiagaraSpriteRendererProperties]}
  {emitterName=Light, rendererClasses [&NiagaraLightRendererProperties, &NiagaraSpriteRendererProperties]}
  {
   bEnabled=false
   emitterName=debug
   rendererClasses [&NiagaraSpriteRendererProperties, &NiagaraMeshRendererProperties]
  }
 ]
 userVariables [
  User.bIsFrontEnd {type=NiagaraBool}
  User.bIsFrontEndPreview {type=NiagaraBool}
 ]
}
