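# Niagara anim trail for the FelineWarrior pickaxe, drawing short-lived pink ribbons along the weapon's animated motion.
# Two Air_Trail emitters use FN AnimTrail modules to build a main and a thinner ribbon, with width curves and randomized hue shift.
# When the system goes inactive, a DoOnce-triggered custom timeline quickly fades their alpha.
# A disabled Lightning_pink emitter holds jittered, flickering ribbons and sprites placed between two skeletal mesh sample points.
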NiagaraSystem /BRCosmetics/Weapons/FORT_Melee/Pickaxe_FelineWarrior/FX/NS_Pickaxe_FelineWarrior_AnimTrail
dependencies {
 usedDataInterfaces [&NiagaraDataInterfaceCurve, &NiagaraDataInterfaceColorCurve, &NiagaraDataInterfaceArrayFloat]
 usedDynamicInputs []
 usedModules [&SystemState]
}
emitters {
 Air_Trail {
  emitterUpdateScript [
   &EmitterState(Life Cycle Mode=::NewEnumerator1, Loop Behavior=::NewEnumerator1, Loop Duration=0.15, UseLoopDelay=true)
   &FN_SpawnAnimTrail_v2()
  ]
  particleSpawnScript [
   /Game/Effects/Niagara/Modules/AnimTrail/FN_InitializeAnimTrail/FN_InitializeAnimTrailBoneIndecies(First Bone or Socket Index=1)
   /Game/Effects/Niagara/Modules/AnimTrail/FN_InitializeAnimTrail/FN_InitializeAnimTrail(
    AnimTrailWidth=1
    Color=<rgba 1 0.15 1 0>
    Lifetime=0.2
    Position Offset:Vector3f=&Lerp_Vector(A=<0 0 0>, B=<0 0 0>, Alpha=&MakeFloatFromInt(Int=$Particles.SpawnGroup))
   )
   &DynamicMaterialParameters(Module.Param0WriteEnabled=false, Module.Param1WriteEnabled=false, Module.Param2WriteEnabled=false, Module.Param3WriteEnabled=false)
  ]
  particleUpdateScript [
   &FN_UpdateAnimTrail_v2()
   &ScaleRibbonWidth(Ribbon Width Scale=&FloatFromCurve(FloatCurve=DataInterface() {propertyValues {Curve {keys [{arriveTangent=-0.8703, time=0.2, value=1}, <at 1 0.8>]}}}))
   &DoOnce(Trigger Condition:NiagaraBool=&self:ScratchDynamicInput_01(ENiagara Execution State:ENiagaraExecutionState=$System.ExecutionState, ENiagara Execution State001:ENiagaraExecutionState=::Inactive))
   &self:Custometimeline(
    Play:NiagaraBool=$Particles.DoOnce.TriggeredOnce
    Play Rate:NiagaraFloat=&Divide_Float(A=1, B=0.08)
    Rewind Mode:ENiagara_TimelineMode=::NewEnumerator1
   ) {module=Custometimeline}
   &ScaleColor(
    Scale Alpha=&FloatFromCurve(FloatCurve=DataInterface() {propertyValues {Curve {keys [{arriveTangent=19.9617, interpMode=RCIM_Cubic, tangentMode=RCTM_User, time=0, value=1}, <at 1 0>]}}})
    Scale RGB=&VectorFromFloat(Value=&FNTimeOfDayScalar(Night Scale=0.35, Morning Scale=0.45, Day Scale=0.6, Evening Scale=0.375))
    ScaleA=true
    ScaleRGB=true
   ) {enabled=false}
   &SpriteFacingAndAlignment(Sprite Facing=$Particles.RibbonFacing, WriteAlignment=false, WriteFacing=true)
   &Color(
    Color=&HueShiftLinearColor(
     Hue=&Lerp_Float(
      A=0
      B=&RandomRangeFloat(Minimum=0.05, Maximum=-0.1)
      Alpha=&FloatFromCurve(FloatCurve=DataInterface() {propertyValues {Curve {keys [<at 0 0 RCIM_Cubic>, {arriveTangent=2.694, interpMode=RCIM_Cubic, leaveTangent=2.694, tangentMode=RCTM_User, time=1, value=1}]}}})
     )
     Saturation=&FloatFromCurve(FloatCurve=DataInterface() {propertyValues {Curve {keys [<at 0 1 RCIM_Cubic>, {arriveTangent=2.9139, interpMode=RCIM_Cubic, leaveTangent=2.9139, tangentMode=RCTM_User, time=1, value=2}]}}})
     Value=1
     Linear Color=&ColorFromCurve(ColorCurve=DataInterface() {propertyValues {
        AlphaCurve {keys [<at 0 1 RCIM_Cubic>,<at 1 0 RCIM_Cubic>]}
        BlueCurve {keys [{arriveTangent=-0.8091, interpMode=RCIM_Cubic, time=0.22, value=1}, {interpMode=RCIM_Cubic, leaveTangent=-1.1334, time=0.8356, value=1}]}
        GreenCurve {keys [
          {arriveTangent=-0.3753, interpMode=RCIM_Cubic, time=0.22, value=0.15}
          {interpMode=RCIM_Cubic, leaveTangent=-0.3429, time=0.8356, value=0.67}
         ]}
        RedCurve {keys [<at 0.22 1 RCIM_Cubic>, {interpMode=RCIM_Cubic, leaveTangent=-1.0652, time=0.8356, value=1}]}
       }})
    )
    Scale Alpha=&Multiply_Float(
     A=0.8
     B=&FloatFromCurve(
      FloatCurve=DataInterface() {propertyValues {Curve {keys [{arriveTangent=-3.0028, interpMode=RCIM_Cubic, time=0, value=1}, <at 1 0 RCIM_Cubic>]}}}
      CurveIndex=$Particles.Custometimeline.NormalizedTime
     )
    )
   )
  ]
  rendererClasses [&NiagaraRibbonRendererProperties]
  renderers [{rendererClass=&NiagaraRibbonRendererProperties}]
 }
 Air_Trail001 {
  emitterUpdateScript [
   &EmitterState(Life Cycle Mode=::NewEnumerator1, Loop Behavior=::NewEnumerator1, Loop Duration=0.15, UseLoopDelay=true)
   &FN_SpawnAnimTrail_v2(AnimTrailWidthMode=::FromSecond)
  ]
  particleSpawnScript [
   &FN_InitializeAnimTrailBoneIndecies(First Bone or Socket Index=1)
   &FN_InitializeAnimTrail(
    AnimTrailWidth=0.1
    Color=<rgba 1 0.2 1 1>
    Lifetime=0.225
    Position Offset:Vector3f=&Lerp_Vector(A=<0 0 0>, B=<0 0 0>, Alpha=&MakeFloatFromInt(Int=$Particles.SpawnGroup))
   )
   &DynamicMaterialParameters(
    Index 0 Param 1=1
    Index 0 Param 2=1
    Index 0 Param 3=1
    Module.Param0WriteEnabled=true
    Module.Param1WriteEnabled=true
    Module.Param2WriteEnabled=true
    Module.Param3WriteEnabled=true
   )
  ]
  particleUpdateScript [
   &FN_UpdateAnimTrail_v2()
   &ScaleRibbonWidth(Ribbon Width Scale=&FloatFromCurve(
     FloatCurve=DataInterface() {propertyValues {Curve {keys [
         {arriveTangent=-0.8703, interpMode=RCIM_Cubic, leaveTangent=-0.8703, tangentMode=RCTM_User, time=0.2, value=1}
         {interpMode=RCIM_Cubic, tangentMode=RCTM_User, time=1, value=0.25}
        ]}}}
     Scale Curve=1.25
    ))
   &DoOnce(Trigger Condition:NiagaraBool=&self:ScratchDynamicInput_01(ENiagara Execution State:ENiagaraExecutionState=$System.ExecutionState, ENiagara Execution State001:ENiagaraExecutionState=::Inactive))
   &self:Custometimeline(
    Play:NiagaraBool=$Particles.DoOnce.TriggeredOnce
    Play Rate:NiagaraFloat=&Divide_Float(A=1, B=0.08)
    Rewind Mode:ENiagara_TimelineMode=::NewEnumerator1
   ) {module=Custometimeline}
   &ScaleColor(
    Scale Alpha=&FloatFromCurve(FloatCurve=DataInterface() {propertyValues {Curve {keys [{arriveTangent=19.9617, time=0, value=1}, <at 1 0>]}}}, Scale Curve=0.5)
    Scale RGB=&VectorFromFloat(Value=&FNTimeOfDayScalar(Night Scale=0.35, Morning Scale=0.45, Day Scale=0.6, Evening Scale=0.375))
    ScaleA=true
    ScaleRGB=true
   ) {enabled=false}
   &SpriteFacingAndAlignment(Sprite Facing=$Particles.RibbonFacing, WriteAlignment=false, WriteFacing=true)
   &Color(
    Color=&HueShiftLinearColor(
     Hue=&Lerp_Float(
      A=0
      B=&RandomRangeFloat(Minimum=0.05, Maximum=-0.1)
      Alpha=&FloatFromCurve(FloatCurve=DataInterface() {propertyValues {Curve {keys [<at 0 0 RCIM_Cubic>, {arriveTangent=2.694, interpMode=RCIM_Cubic, leaveTangent=2.694, tangentMode=RCTM_User, time=1, value=1}]}}})
     )
     Saturation=&FloatFromCurve(FloatCurve=DataInterface() {propertyValues {Curve {keys [<at 0 1 RCIM_Cubic>, {arriveTangent=2.9139, interpMode=RCIM_Cubic, leaveTangent=2.9139, tangentMode=RCTM_User, time=1, value=2}]}}})
     Value=1
     Linear Color=&ColorFromCurve(ColorCurve=DataInterface() {propertyValues {
        AlphaCurve {keys [<at 0 1 RCIM_Cubic>,<at 1 0 RCIM_Cubic>]}
        BlueCurve {keys [{arriveTangent=-0.8091, interpMode=RCIM_Cubic, time=0.22, value=1}, {interpMode=RCIM_Cubic, leaveTangent=-1.1334, time=0.8356, value=1}]}
        GreenCurve {keys [
          {arriveTangent=-0.3753, interpMode=RCIM_Cubic, time=0.22, value=0.15}
          {interpMode=RCIM_Cubic, leaveTangent=-0.3429, time=0.8356, value=0.67}
         ]}
        RedCurve {keys [<at 0.22 1 RCIM_Cubic>, {interpMode=RCIM_Cubic, leaveTangent=-1.0652, time=0.8356, value=1}]}
       }})
    )
    Scale Alpha=&Multiply_Float(
     A=0.25
     B=&FloatFromCurve(
      FloatCurve=DataInterface() {propertyValues {Curve {keys [{arriveTangent=-3.0028, interpMode=RCIM_Cubic, time=0, value=1}, <at 1 0 RCIM_Cubic>]}}}
      CurveIndex=$Particles.Custometimeline.NormalizedTime
     )
    )
   )
  ]
  rendererClasses [&NiagaraRibbonRendererProperties]
  renderers [{rendererClass=&NiagaraRibbonRendererProperties}]
 }
 Lightning_pink {
  bEnabled=false
  emitterUpdateScript [
   &EmitterState(Life Cycle Mode=::NewEnumerator1, Loop Behavior=::NewEnumerator1, Loop Delay=0.03, Loop Duration=0.2, UseLoopDelay=false)
   &SpawnRate(SpawnRate=60, Use Spawn Probability=false)
   &SpawnRate(SpawnGroup=1, SpawnRate=60, Use Spawn Probability=false) {module=SpawnRate002}
   &SpawnRate(SpawnGroup=2, SpawnRate=60, Use Spawn Probability=false) {module=SpawnRate001}
  ]
  particleSpawnScript [
   &InitializeParticle(
    Color=<rgba 5 0 4.5479 1>
    Lifetime=0.125
    Material Random:NiagaraFloat=&SelectFloatFromArray(
     Direct Array Index=&SpawnGroup()
     Float Selection Array=DataInterface() {propertyValues {FloatData [0,0.333,0.666]}}
    )
    Ribbon Width=&RandomRangeFloat(Minimum=100, Maximum=50)
    Ribbon Width Mode=::NewEnumerator1
    Sprite Size Mode=::NewEnumerator3
    UsePositionOffset=false
    Write Material Random=true
   )
   /SetVariables(Particles.Socket Offset:Vector3f=&MakeVector(
     X=&Multiply_Float(A=12, B=&RandomRangeFloat(Minimum=-1, Maximum=1))
     Y=0
     Z=&Multiply_Float(A=12, B=&RandomRangeFloat(Minimum=-1, Maximum=1))
    )) {isSetParameters=true}
   &SkeletalMeshLocation(
    MaskSpawnGroup=false
    Mesh Sampling Type=::NewEnumerator5
    Position Sampling=::NewEnumerator1
    Sampled Position Offset=$Particles.Socket Offset
    Skeletal Mesh=$User.SkeletalMesh
    Socket Sampling Mode=::NewEnumerator5
    Use Sampled Mesh Attribute as Ribbon ID=false
    Write To Particle Position=false
   )
   &SkeletalMeshLocation(
    Bone / Socket Index=1
    MaskSpawnGroup=false
    Mesh Sampling Type=::NewEnumerator5
    Position Sampling=::NewEnumerator1
    Sampled Position Offset=$Particles.Socket Offset
    Skeletal Mesh=$User.SkeletalMesh
    Socket Sampling Mode=::NewEnumerator5
    Use Sampled Mesh Attribute as Ribbon ID=false
    Write To Particle Position=false
   ) {module=SkeletalMeshLocation001}
   /SetVariables(Particles.Position=&LerpPosition(
     A=$Particles.SkeletalMeshLocation001.SampledPosition
     B=$Particles.SkeletalMeshLocation.SampledPosition
     Alpha=&FloatFromCurve(
      FloatCurve=DataInterface() {propertyValues {Curve {keys [{arriveTangent=2.8897, time=0, value=0}, <at 1 1>]}}}
      CurveIndex=&Add_Float(
       A=&RandomRangeFloat(Minimum=-0.1, Maximum=0.1)
       B=&Divide_Float(A=&MakeFloatFromInt(Int=&SpawnGroup()), B=2)
      )
     )
    )) {isSetParameters=true}
   /SetVariables(Particles.Rand:NiagaraFloat=&RandomRangeFloat(Minimum=0, Maximum=1)) {isSetParameters=true}
   /SetVariables(Particles.RibbonID=/Niagara/DynamicInputs/NiagaraID/MakeNiagaraIDFromInt(Index:NiagaraInt32=&SpawnGroup(), Acquire Tag:NiagaraInt32=0)) {isSetParameters=true}
  ]
  particleUpdateScript [
   &ParticleState()
   &DynamicMaterialParameters(
    Index 0 Param 1=1
    Index 0 Param 2=&FloatFromCurve(
     FloatCurve=DataInterface() {propertyValues {Curve {keys [<at 0.33 0 RCIM_Cubic>, {arriveTangent=4.6706, interpMode=RCIM_Cubic, leaveTangent=4.6706, tangentMode=RCTM_User, time=1, value=1}]}}}
     Scale Curve=-1
    )
    Index 0 Param 3=1
    Index 0 Param 4=0.1
    Module.Param0WriteEnabled=true
    Module.Param1WriteEnabled=true
    Module.Param2WriteEnabled=true
    Module.Param3WriteEnabled=true
   )
   &ScaleColor(
    Scale Alpha=&Multiply_Float(
     A=&FloatFromCurve(
      FloatCurve=DataInterface() {propertyValues {Curve {keys [
          {interpMode=RCIM_Cubic, tangentMode=RCTM_User, time=0, value=0}
          {interpMode=RCIM_Cubic, tangentMode=RCTM_User, time=0.5, value=1}
          {tangentMode=RCTM_User, time=1, value=0}
         ]}}}
      CurveIndex=/Game/Effects/Niagara/DynamicInputs/Math/FN_NormalizedParticleCount()
     )
     B=&FNTimeOfDayScalar(Night Scale=0.4, Morning Scale=0.2, Evening Scale=0.4)
    )
    Scale RGB=&VectorFromFloat(Value=&FloatFromCurve(
      FloatCurve=DataInterface() {propertyValues {Curve {keys [
          {arriveTangent=9.0725, interpMode=RCIM_Cubic, leaveTangent=9.0725, tangentMode=RCTM_User, time=0, value=0.25}
          {interpMode=RCIM_Cubic, tangentMode=RCTM_User, time=0.1408, value=0.9847}
          {arriveTangent=-0.0405, interpMode=RCIM_Cubic, leaveTangent=-0.0405, tangentMode=RCTM_User, time=1, value=0.25}
         ]}}}
      CurveIndex=&FN_NormalizedParticleCount()
      Scale Curve=2
     ))
    ScaleA=true
    ScaleRGB=true
   )
   &ScaleColor(
    Scale Mode=::NewEnumerator1
    Scale RGBA=&MakeVector4_FromFloat(Float=&Multiply_Float(
      A=4
      B=&UniformAOrBFloat(A=-1, B=0.5, Evaluation Type=::NewEnumerator0, Recalculate Random Each Loop=true, Override Seed=false)
     ))
    ScaleRGBA=true
   ) {module=ScaleColor001}
   /SetVariables(
    Particles.Ribbon2Width:NiagaraFloat=&FloatFromCurve(
     FloatCurve=DataInterface() {propertyValues {Curve {keys [
         {arriveTangent=3.7122, interpMode=RCIM_Cubic, leaveTangent=3.7122, tangentMode=RCTM_User, time=0, value=0.33}
         {interpMode=RCIM_Cubic, tangentMode=RCTM_User, time=0.33, value=1}
         {arriveTangent=-1.8101, interpMode=RCIM_Cubic, leaveTangent=-1.8101, tangentMode=RCTM_User, time=1, value=0.33}
        ]}}}
     CurveIndex=&FN_NormalizedParticleCount()
     Scale Curve=&FloatFromCurve(
      FloatCurve=DataInterface() {propertyValues {Curve {keys [{interpMode=RCIM_Cubic, tangentMode=RCTM_User, time=0.4, value=1}, {interpMode=RCIM_Cubic, tangentMode=RCTM_User, time=1, value=0.25}]}}}
      Scale Curve=20
     )
    )
    Particles.RibbonFacing=&ReturnMeshOrientationAxis(Mesh Orientation Axis=::ZAxis, Mesh Orientation Quat=$Particles.MeshOrientation)
    Particles.RibbonTwist:NiagaraFloat=&AngleConversion(Angle:NiagaraFloat=45, Angle Input:ENiagara_AngleInput=::NewEnumerator0, Angle Output:ENiagara_AngleInput=::NewEnumerator2)
    Particles.RibbonWidth=&FloatFromCurve(FloatCurve=DataInterface() {propertyValues {Curve {keys [<at 0 1>,<at 1 0.8>]}}}, Scale Curve=45)
   ) {isSetParameters=true}
   &JitterPosition(
    Jitter Amount:NiagaraFloat=&RandomRangeFloat(Minimum=0, Maximum=5, Evaluation Type=::NewEnumerator1)
    Jitter Delay=0.05
   )
   /SetVariables(Particles.DispalyPosition:NiagaraPosition=&AddVectorToPosition(
     Position=$Particles.Position
     Vector=&RandomVector(Vector Scale=&FloatFromCurve(
       FloatCurve=DataInterface() {propertyValues {Curve {keys [<at 0 0>, {arriveTangent=2.7439, interpMode=RCIM_Cubic, leaveTangent=2.7439, tangentMode=RCTM_User, time=1, value=1}]}}}
       Scale Curve=&RandomRangeFloat(Minimum=10, Maximum=20)
      ))
    )) {isSetParameters=true}
   /SetVariables(Particles.MaterialRandom:NiagaraFloat=&Add_Float(
     A=$Particles.Initial.MaterialRandom
     B=&self:ScratchDynamicInput(
      FPS:NiagaraFloat=&RandomRangeFloat(Minimum=7, Maximum=13)
      Time:NiagaraFloat=&Multiply_Float(A=$Engine.System.Age, B=-1)
     )
    )) {isSetParameters=true}
   &self:ScratchModule() {module=ScratchModule}
  ]
  rendererClasses [&NiagaraRibbonRendererProperties, &NiagaraSpriteRendererProperties]
  renderers [
   {rendererClass=&NiagaraRibbonRendererProperties}
   {rendererClass=&NiagaraSpriteRendererProperties, rendererIndex=1}
   {rendererClass=&NiagaraRibbonRendererProperties, rendererIndex=2}
  ]
 }
}
summary {
 systemName=NS_Pickaxe_FelineWarrior_AnimTrail
 emitters [
  {emitterName=Air_Trail, rendererClasses [&NiagaraRibbonRendererProperties]}
  {emitterName=Air_Trail001, rendererClasses [&NiagaraRibbonRendererProperties]}
  {
   bEnabled=false
   emitterName=Lightning_pink
   rendererClasses [&NiagaraRibbonRendererProperties, &NiagaraSpriteRendererProperties]
  }
 ]
 userVariables [
  User.RibbonIDOffset {type=NiagaraInt32}
  User.SkeletalMesh {
   type=NiagaraDataInterfaceSkeletalMesh
   defaultValue {
    struct=NiagaraExt_VariableValue_DataInterface
    value {dataInterface=&self:NiagaraDataInterfaceSkeletalMesh_0}
   }
  }
 ]
}
