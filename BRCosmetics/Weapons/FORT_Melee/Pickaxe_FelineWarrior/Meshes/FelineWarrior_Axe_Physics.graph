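# Physics asset for the FelineWarrior pickaxe axe, covering the bones of its cable.
# Each bone from cable_root through dyn_cable_06 has one capsule body, with offset, tilted capsules on dyn_cable_02 to 05.
# All six dyn_cable bodies are constrained directly to cable_root rather than chained to each other, with tight, uniform swing and twist limits that allow only slight angular movement.
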PhysicsAsset /BRCosmetics/Weapons/FORT_Melee/Pickaxe_FelineWarrior/Meshes/FelineWarrior_Axe_Physics
bodies {
 cable_root [{shapeType=Capsule, length=0.505, radius=0.505, shapeName=cable_root_capsule, rotation=<rot 0 -90 0>}]
 dyn_cable_01 [{shapeType=Capsule, length=0.505, radius=0.505, shapeName=dyn_cable_01_capsule, rotation=<rot 0 -90 0>}]
 dyn_cable_02 [{
   shapeType=Capsule
   length=2.6702
   radius=0.9105
   shapeName=dyn_cable_02_capsule
   center=<-0.5054 0.1132 0.3455>
   rotation=<rot 0 -96.1994 -65.0833>
  }]
 dyn_cable_03 [{
   shapeType=Capsule
   length=1.0078
   radius=0.7883
   shapeName=dyn_cable_03_capsule
   center=<-0.1175 0.0028 0.2258>
   rotation=<rot 0 -89.3421 -87.1072>
  }]
 dyn_cable_04 [{
   shapeType=Capsule
   length=1.2535
   radius=0.8186
   shapeName=dyn_cable_04_capsule
   center=<-0.9683 0.054 0.1114>
   rotation=<rot 0 -84.0146 -84.3787>
  }]
 dyn_cable_05 [{
   shapeType=Capsule
   length=1.7288
   radius=0.6898
   shapeName=dyn_cable_05_capsule
   center=<-0.7502 -0.0906 -0.1025>
   rotation=<rot 0 -86.6302 -84.2061>
  }]
 dyn_cable_06 [{shapeType=Capsule, length=0.505, radius=0.505, shapeName=dyn_cable_06_capsule, rotation=<rot 0 -90 0>}]
}
constraints [
 {bone1Name=dyn_cable_01, bone2Name=cable_root, swing1LimitDegrees=3, swing2LimitDegrees=3, twistLimitDegrees=3}
 {bone1Name=dyn_cable_02, bone2Name=cable_root, swing1LimitDegrees=3, swing2LimitDegrees=3, twistLimitDegrees=3}
 {bone1Name=dyn_cable_03, bone2Name=cable_root, swing1LimitDegrees=3, swing2LimitDegrees=3, twistLimitDegrees=3}
 {bone1Name=dyn_cable_04, bone2Name=cable_root, swing1LimitDegrees=3, swing2LimitDegrees=3, twistLimitDegrees=3}
 {bone1Name=dyn_cable_05, bone2Name=cable_root, swing1LimitDegrees=3, swing2LimitDegrees=3, twistLimitDegrees=3}
 {bone1Name=dyn_cable_06, bone2Name=cable_root, swing1LimitDegrees=3, swing2LimitDegrees=3, twistLimitDegrees=3}
]
