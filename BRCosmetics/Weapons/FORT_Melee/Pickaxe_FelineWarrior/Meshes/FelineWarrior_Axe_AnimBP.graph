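# Animation Blueprint for the FelineWarrior pickaxe that adds secondary physics motion to the axe mesh.
# The AnimGraph runs the incoming pose through a RigidBody node using the axe's physics asset, overridden world gravity, scaled component motion and capped linear velocity and acceleration.
# Each update, the EventGraph sets the simulation blend alpha, using a weaker blend in the front end than elsewhere.
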AnimBlueprint /BRCosmetics/Weapons/FORT_Melee/Pickaxe_FelineWarrior/Meshes/FelineWarrior_Axe_AnimBP
target_skeleton=&/FelineWarrior_Axe_Skeleton
graphs [
 AnimGraph {nodes [
   AnimGraphNode_SubInput_0 = Input Pose()
   n1 = RigidBody(ComponentPose=Local To Component(LocalPose=@AnimGraphNode_SubInput_0), Alpha=Get alpha() -> (alpha:double)) {
    bOverrideWorldGravity=true
    overridePhysicsAsset=&/FelineWarrior_Axe_Physics
    alphaBoolBlend {blendOption=Linear}
    componentLinearAccScale=<0.3 0.2 0.3>
    componentLinearVelScale=<0.2 0.2 0.2>
    overrideWorldGravity {z=-980}
    simSpaceSettings {maxAngularAcceleration=0, maxAngularVelocity=0, maxLinearAcceleration=210, maxLinearVelocity=250, worldAlpha=1}
   }
   Output Pose(Result=Component To Local(ComponentPose=@n1)) {layerGroup=None}
  ]}
 EventGraph {nodes [
   Event BlueprintUpdateAnimation()
   > Set alpha(alpha:double=SelectFloat(A=0.4, B=0.8, bPickA=Get bIsInFrontEnd())) -> (Output_Get:double)
   TryGetPawnOwner()
  ]}
]
variables [alpha {type=number}]
